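# Material function for Helios landscape materials that rebuilds the normal attribute of a mobile material and returns the updated attributes.
# It remaps Normal XY from 0-1 to signed range and derives Z.
# A TransformTangents static switch, on by default, selects a Transform of that normal.
# When off, it approximates world space with a whiteout-style blend against VertexNormalWS, adding XY and multiplying Z without normalizing.
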MaterialFunction /Game/Athena/Helios/Environments/Landscape/Materials/MaterialFunctions_Utilities/MF_MobileDeriveNormals
graph {nodes [
  n1 = GetMaterialAttributes(None=FunctionInput() {inputName=MobileMaterial, inputType=FunctionInput_MaterialAttributes}) {attributeGetTypes [#1]}
  n2 = DeriveNormalZ(InXY=Multiply(A=Add(A=@n1.Normal) {constB=-0.5}) {constB=2})
  n3 = Transform(None=@n2)
  n4 = VertexNormalWS()
  n5 = AppendVector(
   A=Add(A=ComponentMask(None=@n2) {g=true, r=true}, B=ComponentMask(None=@n4) {g=true, r=true})
   B=Multiply(A=ComponentMask(None=@n2) {b=true}, B=ComponentMask(None=@n4) {b=true})
  )
  n6 = FunctionInput(Preview=StaticBool() {value=true}) {bUsePreviewValueAsDefault=true, inputName=TransformTangents, inputType=FunctionInput_StaticBool, sortPriority=1}
  FunctionOutput(None=SetMaterialAttributes!(MaterialAttributes=@n1.MaterialAttributes, Normal=StaticSwitch(True=@n3, False=@n5, Value=@n6) {defaultValue=true}) {attributeSetTypes [#1]}) {outputName=ResultMaterial}
 ]}
