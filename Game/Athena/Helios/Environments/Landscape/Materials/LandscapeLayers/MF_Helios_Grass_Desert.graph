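# Landscape layer material function that height-blends Helios desert grass onto an incoming material, outputting blended materials, masks and a static mesh material.
# The high-end layer's weight is scaled by an inverted, biased TerrainMasks_QuadC red channel, apparently a river mask.
# A separate GrassDesert_NS function builds the mobile layer from terrain and variation masks, with its own blend.
# Shading-path and quality switches select outputs, deriving normals for mobile static meshes.
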MaterialFunction /Game/Athena/Helios/Environments/Landscape/Materials/LandscapeLayers/MF_Helios_Grass_Desert
graph {nodes [
  n1 = NamedRerouteUsage()
  n2 = NamedRerouteUsage()
  n3 = NamedRerouteUsage()
  n4 = NamedRerouteUsage()
  n5 = NamedRerouteUsage()
  n6 = NamedRerouteUsage()
  n7 = MaterialFunctionCall!(InputMaterial=@n2, LayerMaterial=@n3, LayerWeight=@n4, UseContrast=@n5, Contrast=@n6) {materialFunction=/Game/Athena/Helios/Environments/Landscape/Materials/MaterialFunctions_Utilities/MF_HeightBlendLandscapeLayers_OpacityMaskDisplaceCustom}
  n8 = NamedRerouteUsage()
  n9 = StaticBool()
  FunctionOutput(None=ShadingPathSwitch(
    Default=QualitySwitch(Default=@n1, Epic=@n7.BlendedMaterial)
    Mobile=MaterialFunctionCall!(MobileMaterial=@n8, TransformTangents=@n9) {materialFunction=/Game/Athena/Helios/Environments/Landscape/Materials/MaterialFunctions_Utilities/MF_MobileDeriveNormals}.ResultMaterial
   )) {outputName=StaticMeshMaterial, sortPriority=-998}
  FunctionOutput(None=ShadingPathSwitch(Default=NamedRerouteUsage(), Mobile=NamedRerouteUsage())) {outputName=DesertGrass_NoiseMask, sortPriority=-800}
  FunctionOutput(None=@n7.BlendedMaterial) {outputName=BlendedMaterial, sortPriority=-1000}
  n10 = MaterialFunctionCall!(
   InputMaterial=NamedRerouteUsage()
   LayerMaterial=NamedRerouteUsage()
   LayerWeight=NamedRerouteUsage()
   UseContrast=NamedRerouteUsage()
   Contrast=NamedRerouteUsage()
  ) {materialFunction=/Game/Athena/Helios/Environments/Landscape/Materials/MaterialFunctions_Utilities/MF_HeightBlendLandscapeLayers_OpacityMask}
  FunctionOutput(None=@n10.BlendedMaterial) {outputName=MobileBlendedMaterial, sortPriority=-999}
  FunctionOutput(None=ShadingPathSwitch(Default=NamedRerouteUsage(), Mobile=NamedRerouteUsage())) {outputName=LayerMask, sortPriority=-900}
  MaterialFunctionCall(InputMaterial=@n2, LayerMaterial=@n3, LayerWeight=@n4, UseContrast=@n5, Contrast=@n6) {materialFunction=&MF_HeightBlendLandscapeLayers_OpacityMask}
  NamedRerouteDeclaration(None=FunctionInput(Preview=Constant()) {bUsePreviewValueAsDefault=true, inputName=GrassWeight, inputType=FunctionInput_Scalar, sortPriority=14}) {name=GrassWeight, nodeColor=<rgba 0.5 0 1 1>}
  n11 = FunctionInput(Preview=StaticBool()) {
   bUsePreviewValueAsDefault=true
   description="Default = False"
   inputName=UseTangentNormals
   inputType=FunctionInput_StaticBool
   sortPriority=15
  }
  NamedRerouteDeclaration(None=@n11) {name=UseTangentNormalsBool, nodeColor=<rgba 0.5 0 1 1>}
  n12 = SetMaterialAttributes!(MaterialAttributes)
  NamedRerouteDeclaration(None=FunctionInput(Preview=@n12) {bUsePreviewValueAsDefault=true, inputName=Material, inputType=FunctionInput_MaterialAttributes}) {name=InputMaterial, nodeColor=<rgba 0 1 0 1>}
  n13 = MaterialFunctionCall!(LayerWeight=NamedRerouteUsage(), UseTangentNormals=NamedRerouteUsage()) {materialFunction=/Game/Athena/Helios/Environments/Landscape/Materials/QualityLevelLayers/MF_Helios_Grass_Desert_Default}
  NamedRerouteDeclaration(None=@n13.Result) {name=LayerMaterial, nodeColor=<rgba 0 1 0 1>}
  NamedRerouteDeclaration(None=FunctionInput(Preview=Constant()) {bUsePreviewValueAsDefault=true, inputName=LayerWeight, inputType=FunctionInput_Scalar, sortPriority=15}) {name=LayerWeight, nodeColor=<rgba 0.5 0 1 1>}
  NamedRerouteDeclaration(None=@n10.LayerMask) {name=MobileLayerMask, nodeColor=<rgba 0 1 0.7891 1>}
  NamedRerouteDeclaration(None=@n7.LayerMask) {name=LayerMask, nodeColor=<rgba 1 0 0.7266 1>}
  n14 = NamedRerouteDeclaration(None=FunctionInput(Preview=Constant()) {bUsePreviewValueAsDefault=true, inputName=TerrainMasks_QuadC, inputType=FunctionInput_Vector4, sortPriority=30}) {name=TerrainMasks_QuadC, nodeColor=<rgba 0.5 0 1 1>}
  n15 = Add(A=ScalarParameter() {defaultValue=0.6, parameterName=DesertRiverMaskAdd}, B=OneMinus(None=ComponentMask(None=@n14) {r=true}))
  NamedRerouteDeclaration(None=Multiply(A=@n13.HighEnd_LayerWeight AND DesertGrass_Noise, B=Saturate(None=@n15))) {name=HighEnd_LayerWeight, nodeColor=<rgba 0.5 0 1 1>}
  NamedRerouteDeclaration(None=FunctionInput(Preview=Constant()) {bUsePreviewValueAsDefault=true, inputName=VariationMasks Mobile, inputType=FunctionInput_Vector4, sortPriority=36}) {name=VariationMasks Mobile, nodeColor=<rgba 1 0.9131 0 1>}
  n16 = MaterialFunctionCall!(
   TerrainMasks_QuadA=NamedRerouteUsage()
   TerrainMasks_QuadB=NamedRerouteUsage()
   TerrainMasks_QuadC=NamedRerouteUsage()
   TerrainMasks_QuadD=NamedRerouteUsage()
   Layer Weight=NamedRerouteUsage()
   ChapCoastMobile=NamedRerouteUsage()
   VariationMasks Mobile=NamedRerouteUsage()
  ) {materialFunction=/Game/Athena/Helios/Environments/Landscape/Materials/QualityLevelLayers/MF_Helios_GrassDesert_NS}
  NamedRerouteDeclaration(None=@n16.MobileLayerMask) {name=MobileLayerWeight, nodeColor=<rgba 0.6641 1 0 1>}
  NamedRerouteDeclaration(None=FunctionInput(Preview=Constant()) {bUsePreviewValueAsDefault=true, inputName=TerrainMasks_QuadA, inputType=FunctionInput_Vector4, sortPriority=20}) {name=TerrainMasks_QuadA, nodeColor=<rgba 0.5 0 1 1>}
  NamedRerouteDeclaration(None=@n13.StaticMeshMaterial) {name=StaticMeshMaterial, nodeColor=<rgba 0 1 0 1>}
  NamedRerouteDeclaration(None=FunctionInput(Preview=Constant()) {bUsePreviewValueAsDefault=true, inputName=TerrainMasks_QuadD, inputType=FunctionInput_Vector4, sortPriority=35}) {name=TerrainMasks_QuadD, nodeColor=<rgba 0.5 0 1 1>}
  NamedRerouteDeclaration(None=FunctionInput(Preview=Constant()) {bUsePreviewValueAsDefault=true, inputName=TerrainMasks_QuadB, inputType=FunctionInput_Vector4, sortPriority=25}) {name=TerrainMasks_QuadB, nodeColor=<rgba 0.5 0 1 1>}
  NamedRerouteDeclaration(None=FunctionInput(Preview=@n12) {bUsePreviewValueAsDefault=true, inputName=MobileMaterial, inputType=FunctionInput_MaterialAttributes, sortPriority=10}) {name=MobileInputMaterial, nodeColor=<rgba 0 0.5 1 1>}
  NamedRerouteDeclaration(None=@n16.Result) {name=MobileLayerMaterial, nodeColor=<rgba 0 0.5 1 1>}
  NamedRerouteDeclaration(None=StaticBool()) {name=UseContrast, nodeColor=<rgba 0.5 0 1 1>}
  NamedRerouteDeclaration(None=@n16.Result StaticMeshes) {name=MobileStaticMeshes, nodeColor=<rgba 0.982 0.5294 1 1>}
  NamedRerouteDeclaration(None=Constant()) {name=LayerBlendContrast, nodeColor=<rgba 0.5 0 1 1>}
  NamedRerouteDeclaration(None=@n16.MobileNoiseMask) {name=Mobile_NoiseMask, nodeColor=<rgba 0.5 0 1 1>}
  NamedRerouteDeclaration(None=@n13.DesertGrass_Noise) {name=HighEnd_NoiseMask, nodeColor=<rgba 0.5 0 1 1>}
  NamedRerouteDeclaration(None=FunctionInput(Preview=@n12) {bUsePreviewValueAsDefault=true, inputName=ChapCoastMobile, inputType=FunctionInput_MaterialAttributes, sortPriority=10}) {name=ChapCoastMobile, nodeColor=<rgba 0 0.5 1 1>}
  Reroute(None=@n13.HighEnd_LayerWeight AND DesertGrass_Noise)
 ]}
