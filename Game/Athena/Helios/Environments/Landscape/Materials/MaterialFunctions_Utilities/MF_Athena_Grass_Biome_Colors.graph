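# Material function that tints grass base color by terrain biome for the Athena Helios landscape environment.
# RGB channels from MF_TerrainMasks drive chained lerps from a base tint through Biome A, B and C color parameters.
# MatLayerBlend_MultiplyBaseColor multiplies the result into the incoming material attributes.
# The DebugAthenaBiomeolors static switch substitutes the raw mask colors for the tint.
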MaterialFunction /Game/Athena/Helios/Environments/Landscape/Materials/MaterialFunctions_Utilities/MF_Athena_Grass_Biome_Colors
graph {nodes [
  n1 = MaterialFunctionCall!(UseLandscapeCoords, "UseProjection(translucent)") {materialFunction=&/MF_TerrainMasks}
  n2 = VectorParameter() {group=Color, parameterName=BiomeBase_ColorTint, defaultValue=<rgba 1 1 1 0>}
  n3 = VectorParameter() {group=Color, parameterName=BiomeA_ColorTint, defaultValue=<rgba 1 1 1 0>}
  n4 = MaterialFunctionCall(Float4=@n1.FullUV_RGBA) {materialFunction=&BreakOutFloat4Components}
  n5 = LinearInterpolate(
   A=LinearInterpolate(
    A=LinearInterpolate(A=@n2.RGB, B=@n3.RGB, Alpha=@n4.R) {constA=1}
    B=VectorParameter() {group=Color, parameterName=BiomeB_ColorTint, defaultValue=<rgba 1 1 1 0>}.RGB
    Alpha=@n4.G
   ) {constA=1}
   B=VectorParameter() {group=Color, parameterName=BiomeC_ColorTint, defaultValue=<rgba 1 1 1 0>}.RGB
   Alpha=@n4.B
  ) {constA=1}
  FunctionOutput(None=MaterialFunctionCall!(
    Material=FunctionInput() {inputName=SetMaterialAttributes, inputType=FunctionInput_MaterialAttributes}
    NewBaseColor=StaticSwitchParameter(True=@n1.FullUV_RGBA, False=@n5) {parameterName=DebugAthenaBiomeolors}
    Mask
   ) {materialFunction=/Engine/Functions/MaterialLayerFunctions/MatLayerBlend_MultiplyBaseColor}.Blended Material) {outputName=Blended Material}
 ]}
