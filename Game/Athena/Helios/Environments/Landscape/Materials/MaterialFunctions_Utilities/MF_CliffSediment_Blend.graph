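# Material function for Helios cliffs that adds tinted sediment bands to incoming sediment material attributes and combines their normal with the cliff's.
# Two sets of squared triangle-wave bands along offset world Z form a mask, which is also output.
# The mask blends in an overlay-style tint of desaturated base color, using color palettes with optional biome variation.
# Static switches handle landscape normal-space conversion and a debug mask preview.
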MaterialFunction /Game/Athena/Helios/Environments/Landscape/Materials/MaterialFunctions_Utilities/MF_CliffSediment_Blend
graph {nodes [
  Constant3Vector() {constant=<rgba 0 0 0 2500>}
  n1 = NamedRerouteUsage()
  n2 = MaterialFunctionCall!(
   WorldPosition
   Sediment L2_WS | Position Offset_1=Constant3Vector()
   Sediment L2_WS | Scale=Constant() {r=2550}
   Use Wavy Distortion=StaticBool()
   Wave Distortion Intensity=Constant() {r=5000}
   Sediment L2_WS | Wave DistPos Offset=Constant3Vector()
   Wave Distortion Freq=Constant() {r=6000}
  ) {materialFunction=/Game/Global/Base4Layers_2022/MaterialFunctions/MF_Cliffs_SedimentWS_UVs}
  n3 = MaterialFunctionCall!(
   UV_XZ=@n2.XZ
   UV_YZ=@n2.XZ
   T_Diffuse=TextureObject() {texture=/Game/Environments/Helios/Terrain/Textures/T_Helios_RustBucket_Cliff_Strata_01_D}
   T_Specular=TextureObject() {samplerType=SAMPLERTYPE_Masks, texture=/Game/Environments/Helios/Terrain/Textures/T_Helios_RustBucket_Cliff_Strata_01_S}
   T_Normal=TextureObject() {samplerType=SAMPLERTYPE_Normal, texture=/Game/Environments/Helios/Terrain/Textures/T_Helios_RustBucket_Cliff_Strata_01_N}
   PTC=Constant() {desc=XY Projection Transition Contrast, r=0.5}
  ) {materialFunction=/Game/Global/Base4Layers_2022/MaterialFunctions/MF_Cliffs_SedimentWS}
  n4 = GetMaterialAttributes(None=FunctionInput(Preview=@n3.SetMaterialAttributes) {description=MA, inputName=Sediment MA, inputType=FunctionInput_MaterialAttributes}) {attributeGetTypes [
    #1
    #2
    #3
    #4
    #5
   ]}
  n5 = NamedRerouteDeclaration(None=@n4.MaterialAttributes) {name=MA_IN, nodeColor=<rgba 0 1 0.9531 1>}
  n6 = NamedRerouteUsage()
  n7 = NamedRerouteUsage()
  FunctionOutput(None=StaticSwitchParameter(
    True=SetMaterialAttributes!(MaterialAttributes, BaseColor=@n1) {attributeSetTypes [#1]}
    False=SetMaterialAttributes!(MaterialAttributes=@n5, BaseColor=@n6, Normal=@n7) {attributeSetTypes [#1, #3]}
   ) {group=Cliff, parameterName="Sediment_c. | Sediment Mask Preview [Debug]", sortPriority=501}) {description=MA, outputName=MA}
  n8 = Frac(None=Divide(A=WorldPosition().XYZ, B=Constant() {desc=Freq, r=5000}))
  n9 = Multiply(A=OneMinus(None=@n8), B=@n8)
  n10 = MaterialFunctionCall(Color=Multiply(A=@n9, B=@n9)) {materialFunction=&SplitComponents}
  Multiply(A=Add(A=Add(A=@n10.R, B=@n10.R), B=@n10.B), B=Constant() {desc=Intensity, r=10000})
  NamedRerouteDeclaration(None=@n4.BaseColor) {name=Sediment Base Color_IN, nodeColor=<rgba 1 0.3281 0 1>}
  NamedRerouteDeclaration(None=@n4.Roughness) {name=Sediment Roughness IN, nodeColor=<rgba 0.0078 0 1 1>}
  NamedRerouteDeclaration(None=FunctionInput(Preview=StaticBool()) {bUsePreviewValueAsDefault=true, inputName=IsLandscape, inputType=FunctionInput_StaticBool}) {name=IsLandscape_IN, nodeColor=<rgba 0 0.8125 1 1>}
  n11 = Transform(None=NamedRerouteUsage()) {transformSourceType=TRANSFORMSOURCE_World, transformType=TRANSFORM_Tangent}
  n12 = Multiply(A=ComponentMask(None=@n11) {g=true, r=true}, B=NamedRerouteUsage())
  n13 = NamedRerouteUsage()
  n14 = StaticSwitch(
   True=MaterialFunctionCall(Normal=Transform(None=@n13) {transformSourceType=TRANSFORMSOURCE_World, transformType=TRANSFORM_Tangent}, Flatness=Constant() {r=-1}) {materialFunction=&FlattenNormal}.Result
   False=@n13
   Value=NamedRerouteUsage()
  )
  n15 = ComponentMask(None=@n14) {g=true, r=true}
  n16 = ComponentMask(None=@n11) {b=true}
  n17 = ComponentMask(None=@n14) {b=true}
  n18 = Normalize(VectorInput=AppendVector(A=Add(A=@n12, B=@n15), B=Multiply(A=@n16, B=@n17)))
  NamedRerouteDeclaration(None=StaticSwitch(True=Transform(None=@n18), False=@n18, Value=NamedRerouteUsage())) {name=Normal_OUT, nodeColor=<rgba 0.9453 0 1 1>}
  NamedRerouteDeclaration(None=FunctionInput(Preview=Constant() {r=1}) {bUsePreviewValueAsDefault=true, inputName=Sediment Normal Intensity, inputType=FunctionInput_Scalar}) {name=SedimentNIntensity_IN, nodeColor=<rgba 0 0.1094 1 1>}
  NamedRerouteDeclaration(None=@n4.Normal) {name=Sediment Normal_IN, nodeColor=<rgba 1 0 0.75 1>}
  n19 = TextureSample() {texture=/Game/Environments/Helios/Terrain/Textures/T_Helios_RustBucket_Cliff_01_D}
  n20 = TextureSample() {samplerType=SAMPLERTYPE_Masks, texture=/Game/Environments/Helios/Terrain/Textures/T_Helios_RustBucket_Cliff_01_S}
  n21 = SetMaterialAttributes!(
   MaterialAttributes
   BaseColor=@n19.RGB
   Roughness=@n20.B
   Specular=@n20.B
   Normal=TextureSample() {samplerType=SAMPLERTYPE_Normal, texture=/Game/Environments/Helios/Terrain/Textures/T_Helios_RustBucket_Cliff_02_N}.RGB
   Metallic=@n20.B
   Opacity=@n19.RGB
  ) {attributeSetTypes [
    #1
    #2
    #6
    #3
    #4
    #5
   ]}
  NamedRerouteDeclaration(None=GetMaterialAttributes(None=FunctionInput(Preview=@n21) {description=MA, inputName=Cliffs MA, inputType=FunctionInput_MaterialAttributes}) {attributeGetTypes [#5, #3]}.Normal) {name=Baked L1 Normal_IN, nodeColor=<rgba 1 0.6094 0 1>}
  n22 = StaticBoolParameter() {
   desc=If eneabled uses the color Biom Mask to tint Sediment Layers using two color palletes
   group=Cliff
   parameterName=Sediment_b | Color Biom Mask Use
   sortPriority=501
  }
  n23 = VectorParameter() {group=Cliff, parameterName=Sediment_b | Color Biom 2, sortPriority=501, defaultValue=<rgba 0.05 0.0833 0.1 1>}
  n24 = VectorParameter() {group=Cliff, parameterName=Sediment_b | Color Biom 1, sortPriority=501, defaultValue=<rgba 0.54 0.8467 1 1>}
  n25 = ScalarParameter() {defaultValue=-0.4, group=Cliff, parameterName=Sediment_b | Color Bias, sortPriority=501}
  n26 = NamedRerouteUsage()
  n27 = Saturate(None=Multiply(
    A=Add(A=@n26, B=@n25) {constB=-0.75}
    B=ScalarParameter() {defaultValue=4, group=Cliff, parameterName=Sediment_b | Color Bias Contrast, sortPriority=501}
   ) {constB=4})
  n28 = LinearInterpolate(A=@n23.RGB, B=@n24.RGB, Alpha=@n27)
  n29 = NamedRerouteUsage()
  n30 = VectorParameter() {group=Cliff, parameterName=Sediment_b | Color 2, sortPriority=501, defaultValue=<rgba 0.05 0.0833 0.1 1>}
  n31 = VectorParameter() {group=Cliff, parameterName=Sediment_b | Color 1, sortPriority=501, defaultValue=<rgba 0.54 0.8467 1 1>}
  n32 = LinearInterpolate(A=@n30.RGB, B=@n31.RGB, Alpha=@n27)
  n33 = StaticSwitch(
   True=StaticSwitch(
    True=@n28
    False=LinearInterpolate(A=@n32, B=@n28, Alpha=@n29)
    Value=StaticBoolParameter() {group=Cliff, parameterName=Sediment_b | Color Biom Only, sortPriority=501}
   )
   False=@n32
   Value=@n22
  )
  n34 = Desaturation(None=NamedRerouteUsage(), Fraction=ScalarParameter() {group=Cliff, parameterName=Sediment_b | Desaturate Base Color, sortPriority=501})
  n35 = Step(Y=@n34) {constX=0.5}
  n36 = Multiply(A=Multiply(A=Multiply(A=@n33, B=@n34) {constB=2}) {constB=2}, B=@n35)
  n37 = Multiply(A=OneMinus(None=@n33), B=OneMinus(None=@n34)) {constB=2}
  n38 = Multiply(A=OneMinus(None=Multiply(A=@n37) {constB=2}), B=OneMinus(None=@n35))
  n39 = NamedRerouteUsage()
  n40 = ScalarParameter() {
   defaultValue=911
   desc="For more variety between layers use values that are not multiple between layers (prime numbers are perfect for this)"
   group=Cliff
   parameterName=Sediment_c | Color Bands 1 Size
   sortPriority=501
  }
  n41 = Subtract(
   A=Subtract(A=WorldPosition().Z, B=ScalarParameter() {group=Cliff, parameterName=Sediment_c | Color Band Z Offset, sortPriority=501})
   B=NamedRerouteUsage()
  )
  n42 = Frac(None=Divide(A=@n41, B=@n40) {constB=500})
  n43 = Abs(None=Add(A=Multiply(A=@n42) {constB=2}) {constB=-1})
  n44 = Multiply(A=@n43, B=@n43)
  n45 = ScalarParameter() {defaultValue=3, group=Cliff, parameterName=Sediment_c | Color Bands 1 Contrast, sortPriority=501}
  n46 = ScalarParameter() {
   defaultValue=211
   desc="For more variety between layers use values that are not multiple between layers (prime numbers are perfect for this)"
   group=Cliff
   parameterName=Sediment_c | Color Bands 2 Size
   sortPriority=501
  }
  n47 = Frac(None=Divide(A=@n41, B=@n46) {constB=500})
  n48 = Abs(None=Add(A=Multiply(A=@n47) {constB=2}) {constB=-1})
  n49 = Add(
   A=Multiply(
    A=Multiply(A=@n48, B=@n48)
    B=ScalarParameter() {defaultValue=1, group=Cliff, parameterName=Sediment_c | Color Bands 2 Contrast, sortPriority=501}
   ) {constB=0.2}
   B=ScalarParameter() {defaultValue=-0.1, group=Cliff, parameterName=Sediment_c | Color Bands 2 Value Bias, sortPriority=501}
  )
  n50 = Add(
   A=Add(
    A=Multiply(A=@n44, B=@n45) {constB=4}
    B=ScalarParameter() {defaultValue=-0.15, group=Cliff, parameterName=Sediment_c | Color Bands 1 Value Bias, sortPriority=501}
   )
   B=@n49
  )
  n51 = StaticSwitchParameter(True=Max(A=@n50, B=@n49), False=@n50) {
   desc="if ON: Bands 2 values that are higher than it`s sum with Bands 1 will allways be visible max((b1+b2),b2)\r\nif OFF: Just adds bands 1 and 2 (b1+b2)"
   group=Cliff
   parameterName=Sediment_c | Color Bands 2 Always Visible
   sortPriority=501
  }
  n52 = Subtract(
   A=@n51
   B=Multiply(A=OneMinus(None=@n39), B=ScalarParameter() {defaultValue=4, group=Cliff, parameterName=Sediment_c | Color AO Distortion, sortPriority=501})
  )
  n53 = FunctionOutput(None=Saturate(None=Add(A=@n51, B=@n52))) {outputName=Sediment Bands Mask, sortPriority=100}
  NamedRerouteDeclaration(None=LinearInterpolate(A=NamedRerouteUsage(), B=Add(A=@n36, B=@n38), Alpha=NamedRerouteDeclaration(None=@n53) {name=Sediment Bands, nodeColor=<rgba 1 0 0.6094 1>})) {name=Base Color OUT, nodeColor=<rgba 0 0.2266 1 1>}
  NamedRerouteDeclaration(None=FunctionInput(Preview=@n2.Sediment Wave Distortion) {inputName=Sediment Wave Distiortion}) {name=Sediment Wave Distortion, nodeColor=<rgba 0 1 0.6484 1>}
  NamedRerouteDeclaration(None=@n4.Opacity) {name=Sediment Height IN, nodeColor=<rgba 1 0 0.7031 1>}
  NamedRerouteDeclaration(None=@n4.Metallic) {name=SedimentColorMask_IN, nodeColor=<rgba 0 1 0.0859 1>}
  NamedRerouteDeclaration(None=FunctionInput(Preview=Constant() {r=1}) {bUsePreviewValueAsDefault=true, inputName=Biom Mask, inputType=FunctionInput_Scalar, sortPriority=1}) {name=Biom Mask_IN, nodeColor=<rgba 1 0 0.2344 1>}
  NamedRerouteDeclaration(None=FunctionInput(Preview=@n3.Sediment Normal B Mask) {inputName=Sediment Normal B Mask, inputType=FunctionInput_Vector2}) {name=Sediment Normal B Mask, nodeColor=<rgba 0 0.5547 1 1>}
  Reroute()
  n54 = TextureSample() {texture=&T_Helios_RustBucket_Cliff_Strata_01_D}
  n55 = TextureSample() {samplerType=SAMPLERTYPE_Masks, texture=&T_Helios_RustBucket_Cliff_Strata_01_S}
  SetMaterialAttributes(
   BaseColor=@n54.RGB
   Roughness=@n55.B
   Specular=@n55.B
   Normal=TextureSample() {samplerType=SAMPLERTYPE_Normal, texture=&T_Helios_RustBucket_Cliff_Strata_01_N}.RGB
   Metallic=@n55.B
   Opacity=@n54.RGB
  ) {attributeSetTypes [
    #1
    #2
    #6
    #3
    #4
    #5
   ]}
 ]}
