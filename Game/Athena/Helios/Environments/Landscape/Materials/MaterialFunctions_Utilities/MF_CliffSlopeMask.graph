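# MF_CliffSlopeMask is a landscape utility material function that outputs a slope mask for steep, cliff-like surfaces.
# It takes the absolute vertical (Z) component of the world-space vertex normal and inverts it, so horizontal faces read zero and vertical faces read high.
# A fixed multiplier boosts the result before saturation, so near-vertical slopes reach full strength.
# It has no inputs, and using vertex normals keeps the mask independent of normal maps.
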MaterialFunction /Game/Athena/Helios/Environments/Landscape/Materials/MaterialFunctions_Utilities/MF_CliffSlopeMask
graph {nodes [
  n1 = VertexNormalWS()
  n2 = OneMinus(None=Abs(None=ComponentMask(None=@n1) {b=true}))
  n3 = Constant() {r=1.4}
  FunctionOutput(None=Saturate(None=Multiply(A=@n2, B=@n3)))
 ]}
