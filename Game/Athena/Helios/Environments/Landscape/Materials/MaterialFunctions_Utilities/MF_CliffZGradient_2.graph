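# Cliff material function for Helios terrain that tints the incoming attributes' base color with sediment layer colors.
# Two banded masks are built from tilted world-space height, with optional wave distortion and roughness-driven offset, or Spec_G can serve as the mask.
# The mask selects between palette colors, optionally biome-masked, which are overlay-blended into base color.
# A static debug switch can preview the sediment mask as emissive.
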MaterialFunction /Game/Athena/Helios/Environments/Landscape/Materials/MaterialFunctions_Utilities/MF_CliffZGradient_2
graph {nodes [
  n1 = NamedRerouteUsage()
  n2 = TextureSample() {samplerType=SAMPLERTYPE_Masks, texture=/Game/Environments/Helios/Terrain/Textures/T_Helios_RustBucket_Cliff_01_S}
  n3 = SetMaterialAttributes!(
   MaterialAttributes
   BaseColor=TextureSample() {texture=/Game/Environments/Helios/Terrain/Textures/T_Helios_RustBucket_Cliff_01_D}.RGB
   Roughness=@n2.B
   Specular=@n2.B
   Normal=TextureSample() {samplerType=SAMPLERTYPE_Normal, texture=/Game/Environments/Helios/Terrain/Textures/T_Helios_RustBucket_Cliff_02_N}.RGB
  ) {attributeSetTypes [
    #1
    #2
    #3
    #4
   ]}
  n4 = GetMaterialAttributes(None=FunctionInput(Preview=@n3) {description=MA, inputName=MA, inputType=FunctionInput_MaterialAttributes}) {attributeGetTypes [#1, #2, #4]}
  n5 = SetMaterialAttributes!(MaterialAttributes=@n4.MaterialAttributes, BaseColor=NamedRerouteUsage()) {attributeSetTypes [#1]}
  FunctionOutput(None=StaticSwitchParameter(True=SetMaterialAttributes!(MaterialAttributes, EmissiveColor=@n1) {attributeSetTypes [#5]}, False=@n5) {group=Cliff, parameterName="Sediment_c. | Sediment Mask Preview [Debug]", sortPriority=501}) {description=MA, outputName=MA}
  NamedRerouteDeclaration(None=@n4.BaseColor) {name=Base Color_IN, nodeColor=<rgba 1 0.3281 0 1>}
  NamedRerouteDeclaration(None=@n4.Roughness) {name=Roughness IN, nodeColor=<rgba 0.0078 0 1 1>}
  n6 = StaticBoolParameter() {
   desc=If eneabled uses the color Biom Mask to tint Sediment Layers using two color palletes
   group=Cliff
   parameterName=Sediment_b | Color Biom Mask Use
   sortPriority=501
  }
  n7 = VectorParameter() {group=Cliff, parameterName=Sediment_b | Color Biom 2, sortPriority=501, defaultValue=<rgba 0.05 0.0833 0.1 1>}
  n8 = VectorParameter() {group=Cliff, parameterName=Sediment_b | Color Biom 1, sortPriority=501, defaultValue=<rgba 0.54 0.8467 1 1>}
  n9 = ScalarParameter() {defaultValue=-0.4, group=Cliff, parameterName=Sediment_b | Color Bias, sortPriority=501}
  n10 = ScalarParameter() {
   defaultValue=211
   desc="For more variety between layers use values that are not multiple between layers (prime numbers are perfect for this)"
   group=Cliff
   parameterName=Sediment_c | Bands 2 Size
   sortPriority=501
  }
  n11 = ScalarParameter() {defaultValue=0.16, group=Cliff, parameterName=Sediment_b | X Angle, sortPriority=501}
  n12 = MaterialFunctionCall(Color=Divide(A=WorldPosition().XYZ, B=ScalarParameter() {defaultValue=1, group=Cliff, parameterName=Sediment_a | Global Tiling, sortPriority=501})) {materialFunction=&SplitComponents}
  n13 = ScalarParameter() {defaultValue=0.16, group=Cliff, parameterName=Sediment_b | Y Angle, sortPriority=501}
  n14 = Add(A=Add(A=Multiply(A=@n12.R, B=@n11), B=Multiply(A=@n12.G, B=@n13)), B=@n12.B)
  n15 = Frac(None=Divide(A=@n12.RGB, B=ScalarParameter() {defaultValue=10000, group=Cliff, parameterName=Sediment_a | Wave Distortion Tiling, sortPriority=501}))
  n16 = Multiply(A=OneMinus(None=@n15), B=@n15)
  n17 = MaterialFunctionCall(Color=Multiply(A=@n16, B=@n16)) {materialFunction=&SplitComponents}
  n18 = Add(A=Add(A=@n17.R, B=@n17.R), B=@n17.B)
  n19 = ScalarParameter() {defaultValue=10000, group=Cliff, parameterName=Sediment_a | Wave Distortion Intensity, sortPriority=501}
  n20 = StaticSwitchParameter(True=Subtract(A=@n14, B=Multiply(A=@n18, B=@n19)), False=@n14) {group=Cliff, parameterName=Sediment_a | Wave Distortion, sortPriority=501}
  n21 = ScalarParameter() {
   defaultValue=1
   desc="this parameter should only distort the bands. If it is also offseting the bands position try finding a value that neutralizes this translation in \"Sediment_b | Roughness Mask Value\" parameter"
   group=Cliff
   parameterName=Sediment_b | Roughness Distortion
   sortPriority=501
  }
  n22 = ScalarParameter() {
   defaultValue=-0.8
   desc="Use this to define a value so Roughness distortion won`t offset bands position. "
   group=Cliff
   parameterName=Sediment_b | Roughness Mask Value
   sortPriority=501
  }
  n23 = Add(A=@n22, B=NamedRerouteUsage()) {constB=0.5}
  n24 = Subtract(A=@n23, B=Multiply(A=NamedRerouteUsage(), B=Add(B=@n22) {constA=1}))
  n25 = NamedRerouteUsage()
  n26 = Multiply(
   A=NamedRerouteDeclaration(None=StaticSwitch(True=@n24, False=@n23, Value=@n25)) {name=Sediment_Roughness, nodeColor=<rgba 1 0.4688 0 1>}
   B=@n21
  )
  n27 = Frac(None=Add(A=Divide(A=@n20, B=@n10) {constB=500}, B=@n26))
  n28 = Multiply(A=@n27, B=Multiply(A=@n27, B=@n27))
  n29 = Multiply(
   A=Multiply(A=OneMinus(None=@n28), B=@n28)
   B=ScalarParameter() {defaultValue=1.3, group=Cliff, parameterName=Sediment_c | Bands 2 Contrast, sortPriority=501}
  ) {constB=0.2}
  n30 = ScalarParameter() {
   defaultValue=-0.1
   desc="positive values: bias this band wave patten to be more visible in the final blend\r\nnegative values: bias this band wave patten to be less visible and even subtract areas in the final blend"
   group=Cliff
   parameterName=Sediment_c | Bands 2 Value Bias
   sortPriority=501
  }
  n31 = ScalarParameter() {
   defaultValue=911
   desc="For more variety between layers use values that are not multiple between layers (prime numbers are perfect for this)"
   group=Cliff
   parameterName=Sediment_c | Bands 1 Size
   sortPriority=501
  }
  n32 = Frac(None=Add(A=@n26, B=Divide(A=@n20, B=@n31) {constB=500}))
  n33 = Multiply(A=@n32, B=Multiply(A=@n32, B=@n32))
  n34 = OneMinus(None=@n33)
  n35 = ScalarParameter() {
   defaultValue=-0.15
   desc="positive values: bias this band wave patten to be more visible in the final blend\r\nnegative values: bias this band wave patten to be less visible and even subtract areas in the final blend"
   group=Cliff
   parameterName=Sediment_c | Bands 1 Value Bias
   sortPriority=501
  }
  n36 = Add(
   A=Multiply(A=Multiply(A=@n34, B=@n33), B=ScalarParameter() {defaultValue=3, group=Cliff, parameterName=Sediment_c | Bands 1 Contrast, sortPriority=501}) {constB=4}
   B=@n35
  )
  n37 = Max(A=Add(A=Add(A=@n29, B=@n30), B=@n36), B=@n36)
  n38 = NamedRerouteUsage()
  n39 = ScalarParameter() {group=Cliff, parameterName=Sediment_b | Spec_G Color Opacity, sortPriority=501}
  n40 = StaticBoolParameter() {
   desc="Activate this to mitigate Roughness  being too agressive on some areas"
   group=Cliff
   parameterName=Sediment_b | Spec_G Adjustments
   sortPriority=501
  }
  n41 = StaticSwitch(
   True=Subtract(A=@n37, B=Multiply(A=@n38, B=@n39))
   False=@n37
   Value=NamedRerouteDeclaration(None=@n40) {name=Use Spec_G, nodeColor=<rgba 0 0.5312 1 1>}
  )
  n42 = StaticSwitch(
   True=NamedRerouteUsage()
   False=NamedRerouteDeclaration(None=Saturate(None=@n41)) {name=Sedimentary Layers Mask, nodeColor=<rgba 1 0 0.6094 1>}
   Value=StaticBoolParameter() {defaultValue=true, group=Cliff, parameterName=Sediment_a | Spec_G is Sediment Mask, sortPriority=501}
  )
  n43 = Saturate(None=Multiply(
    A=Add(A=@n42, B=@n9) {constB=-0.75}
    B=ScalarParameter() {defaultValue=4, group=Cliff, parameterName=Sediment_b | Color Bias Contrast, sortPriority=501}
   ) {constB=4})
  n44 = LinearInterpolate(A=@n7.RGB, B=@n8.RGB, Alpha=@n43)
  n45 = VectorParameter() {group=Cliff, parameterName=Sediment_b | Color 2, sortPriority=501, defaultValue=<rgba 0.05 0.0833 0.1 1>}
  n46 = VectorParameter() {group=Cliff, parameterName=Sediment_b | Color 1, sortPriority=501, defaultValue=<rgba 0.54 0.8467 1 1>}
  n47 = LinearInterpolate(A=@n45.RGB, B=@n46.RGB, Alpha=@n43)
  n48 = LinearInterpolate(A=@n47, B=@n44, Alpha=NamedRerouteUsage())
  n49 = StaticSwitch(
   True=StaticSwitch(True=@n44, False=@n48, Value=StaticBoolParameter() {group=Cliff, parameterName=Sediment_b | Color Biom Only, sortPriority=501})
   False=@n47
   Value=@n6
  )
  n50 = NamedRerouteUsage()
  n51 = Step(Y=@n50) {constX=0.5}
  n52 = Multiply(A=Multiply(A=Multiply(A=@n49, B=@n50) {constB=2}) {constB=2}, B=@n51)
  n53 = Multiply(A=OneMinus(None=@n49), B=OneMinus(None=@n50)) {constB=2}
  n54 = Multiply(A=OneMinus(None=Multiply(A=@n53) {constB=2}), B=OneMinus(None=@n51))
  n55 = Multiply(
   A=Add(A=@n42, B=ScalarParameter() {group=Cliff, parameterName=Sediment_b | SpecG Value Bias, sortPriority=501})
   B=ScalarParameter() {defaultValue=1, group=Cliff, parameterName=Sediment_c | SpecG Contrast, sortPriority=501}
  )
  NamedRerouteDeclaration(None=LinearInterpolate(A=NamedRerouteUsage(), B=Add(A=@n52, B=@n54), Alpha=Saturate(None=@n55))) {name=Base Color OUT, nodeColor=<rgba 0 0.2266 1 1>}
  NamedRerouteDeclaration(None=@n4.Normal) {name=Normal IN, nodeColor=<rgba 1 0.7266 0 1>}
  n56 = FunctionInput(Preview=@n2.G) {desc=Used to Cancel Roughness Distortion Where Spec_G is White, inputName=Spec_G Mask, inputType=FunctionInput_Scalar, sortPriority=3}
  NamedRerouteDeclaration(None=@n56) {name=Spec G Mask_IN, nodeColor=<rgba 0 1 0.0859 1>}
  NamedRerouteDeclaration(None=FunctionInput(Preview=Constant() {r=1}) {bUsePreviewValueAsDefault=true, inputName=Biom Mask, inputType=FunctionInput_Scalar, sortPriority=1}) {name=Biom Mask_IN, nodeColor=<rgba 1 0 0.2344 1>}
  NamedRerouteUsage()
  Reroute(None=@n48)
  StaticSwitchParameter(True=SetMaterialAttributes(EmissiveColor=NamedRerouteUsage()) {attributeSetTypes [#5]}, False=@n5) {group=Cliff, parameterName="Sediment_e. | Roughness Preview [Debug]", sortPriority=501}
 ]}
