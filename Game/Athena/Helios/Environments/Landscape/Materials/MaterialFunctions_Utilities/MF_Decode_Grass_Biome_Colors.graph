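# Material function that decodes the terrain biome mask into per-biome color tints for grass and multiplies them onto the input material's base color.
# It splits MF_TerrainMasks' RGB into White, Black, Red, Green and Blue masks using a channel product, inverted sum and subtraction.
# Five tint parameters (Base, A–D) are mask-weighted and summed, then applied through MatLayerBlend_MultiplyBaseColor.
# A static debug switch multiplies by the raw mask instead.
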MaterialFunction /Game/Athena/Helios/Environments/Landscape/Materials/MaterialFunctions_Utilities/MF_Decode_Grass_Biome_Colors
graph {nodes [
  n1 = SetMaterialAttributes!(MaterialAttributes, BaseColor=Constant() {r=1}) {attributeSetTypes [#1]}
  MaterialExpressionReroute_4 = Reroute(None=MaterialFunctionCall!(UseLandscapeCoords, "UseProjection(translucent)") {materialFunction=&/MF_TerrainMasks}.FullUV_RGBA)
  n2 = Add(
   A=Multiply(A=VectorParameter() {group=Color, parameterName=BiomeBase_ColorTint, defaultValue=<rgba 1 1 1 1>}.RGB, B=NamedRerouteUsage())
   B=Multiply(A=VectorParameter() {group=Color, parameterName=BiomeA_ColorTint, defaultValue=<rgba 1 1 1 1>}.RGB, B=NamedRerouteUsage())
  )
  n3 = Multiply(A=VectorParameter() {group=Color, parameterName=BiomeB_ColorTint, defaultValue=<rgba 1 1 1 1>}.RGB, B=NamedRerouteUsage())
  n4 = VectorParameter() {group=Color, parameterName=BiomeC_ColorTint, defaultValue=<rgba 1 1 1 1>}
  n5 = NamedRerouteUsage()
  n6 = Add(
   A=Add(A=Add(A=@n2, B=@n3), B=Multiply(A=@n4.RGB, B=@n5))
   B=Multiply(A=VectorParameter() {group=Color, parameterName=BiomeD_ColorTint, defaultValue=<rgba 1 1 1 1>}.RGB, B=NamedRerouteUsage())
  )
  FunctionOutput(None=MaterialFunctionCall!(
    Material=FunctionInput(Preview=@n1) {bUsePreviewValueAsDefault=true, inputName=SetMaterialAttributes, inputType=FunctionInput_MaterialAttributes}
    NewBaseColor=StaticSwitchParameter(True=@MaterialExpressionReroute_4, False=@n6) {parameterName=DebugAthenaBiomeolors}
    Mask
   ) {materialFunction=/Engine/Functions/MaterialLayerFunctions/MatLayerBlend_MultiplyBaseColor}.Blended Material) {outputName=Blended Material}
  n7 = Convert(None=@MaterialExpressionReroute_4) {
   nodeName=Break Float3
   convertInputs [{
     type=Vector3
     expressionInput {
      expression=&self:MaterialExpressionReroute_4
      inputName=None
      mask=0
      maskA=0
      maskB=0
      maskG=0
      maskR=0
      outputIndex=0
     }
    }]
   convertOutputs [{type=Scalar}, {type=Scalar}, {type=Scalar}]
  }
  n8 = Multiply(A=@n7, B=@n7)
  n9 = Saturate(None=Power(Base=Multiply(A=@n8, B=@n7), Exp=Divide() {constB=3}) {constExponent=0.3333})
  n10 = OneMinus(None=Saturate(None=MaterialFunctionCall(f3=@MaterialExpressionReroute_4) {materialFunction=&AddComponents}.f3))
  MaterialExpressionSaturate_5 = Saturate(None=Subtract(A=@MaterialExpressionReroute_4, B=NamedRerouteUsage()))
  n11 = Convert(None=@MaterialExpressionSaturate_5) {
   nodeName=Break Float3
   convertInputs [{
     type=Vector3
     expressionInput {
      expression=&self:MaterialExpressionSaturate_5
      inputName=None
      mask=0
      maskA=0
      maskB=0
      maskG=0
      maskR=0
      outputIndex=0
     }
    }]
   convertOutputs [{type=Scalar}, {type=Scalar}, {type=Scalar}]
  }
  n12 = Add(
   A=Add(
    A=NamedRerouteDeclaration(None=@n9) {name=White, nodeColor=<rgba 1 1 1 1>}
    B=NamedRerouteDeclaration(None=@n10) {name=Black, nodeColor=<rgba 0 0 0 1>}
   )
   B=NamedRerouteDeclaration(None=@n11) {name=Red, nodeColor=<rgba 1 0 0 1>}
  )
  n13 = NamedRerouteDeclaration(None=@n11) {name=Green, nodeColor=<rgba 0 1 0 1>}
  Subtract(A=Add(A=Add(A=@n12, B=@n13), B=NamedRerouteDeclaration(None=@n11) {name=Blue, nodeColor=<rgba 0 0 1 1>})) {constB=0.5}
 ]}
