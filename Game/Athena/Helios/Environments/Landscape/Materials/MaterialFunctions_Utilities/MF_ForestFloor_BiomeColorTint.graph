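# Landscape material function that tints forest floor base color by biome and returns the material attributes with that color replaced.
# Per-biome color parameters multiply base color by RGB and alpha as intensity, then saturate and blend through lerps, with biome masks broken out of terrain mask quads.
# Only the Olympus blend feeds the output color; the Chaparral and Boreal chain is left unconnected, as are the QuadA and QuadC inputs.
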MaterialFunction /Game/Athena/Helios/Environments/Landscape/Materials/MaterialFunctions_Utilities/MF_ForestFloor_BiomeColorTint
graph {
 properties {libraryCategoriesText [FN-Landscape]}
 nodes [
  FunctionInput() {bUsePreviewValueAsDefault=true, inputName=TerrainMasks_QuadA, inputType=FunctionInput_Vector4, sortPriority=10}
  FunctionInput() {bUsePreviewValueAsDefault=true, inputName=TerrainMasks_QuadC, inputType=FunctionInput_Vector4, sortPriority=30}
  FunctionOutput(None=SetMaterialAttributes!(MaterialAttributes=NamedRerouteUsage(), BaseColor=NamedRerouteUsage()) {attributeSetTypes [#1]})
  n1 = VectorParameter() {group=BiomeColors, parameterName=ForestFloorColor_Boreal_Forest, defaultValue=<rgba 1 1 1 1>}
  n2 = Multiply(A=Multiply(A=NamedRerouteUsage(), B=@n1.RGB), B=@n1.A)
  n3 = NamedRerouteUsage()
  n4 = VectorParameter() {group=BiomeColors, parameterName=ForestFloorColor_Chaparral_Coast, defaultValue=<rgba 1 1 1 1>}
  n5 = Multiply(A=@n3, B=@n4.RGB)
  n6 = LinearInterpolate(A=@n3, B=Saturate(None=Multiply(A=@n5, B=@n4.A)), Alpha=NamedRerouteUsage())
  n7 = NamedRerouteUsage()
  n8 = VectorParameter() {group=BiomeColors, parameterName=ForestFloorColor_Boreal_Tundra, defaultValue=<rgba 1 1 1 1>}
  n9 = Saturate(None=Multiply(A=Multiply(A=@n7, B=@n8.RGB), B=@n8.A))
  n10 = NamedRerouteUsage()
  n11 = VectorParameter() {group=BiomeColors, parameterName=ForestFloorColor_Boreal_Mountain, defaultValue=<rgba 1 1 1 1>}
  n12 = Multiply(A=NamedRerouteUsage(), B=@n11.RGB)
  LinearInterpolate(
   A=LinearInterpolate(A=LinearInterpolate(A=@n6, B=@n9, Alpha=@n10), B=Saturate(None=@n2), Alpha=NamedRerouteUsage())
   B=Saturate(None=Multiply(A=@n12, B=@n11.A))
   Alpha=NamedRerouteUsage()
  )
  n13 = MaterialFunctionCall(Float4=FunctionInput() {bUsePreviewValueAsDefault=true, inputName=TerrainMasks_QuadB, inputType=FunctionInput_Vector4, sortPriority=20}) {materialFunction=&BreakOutFloat4Components}
  NamedRerouteDeclaration(None=@n13.R) {name=Biome_Chaparral_Coast, nodeColor=<rgba 0.5 0 1 1>}
  NamedRerouteDeclaration(None=@n13.G) {name=Biome_Boreal_Forest, nodeColor=<rgba 0.5 0 1 1>}
  n14 = NamedRerouteUsage()
  n15 = NamedRerouteUsage()
  n16 = VectorParameter() {group=BiomeColors, parameterName=ForestFloorColor_Olympus, defaultValue=<rgba 1 0.639 0.3 3>}
  n17 = Saturate(None=Multiply(A=Multiply(A=@n15, B=@n16.RGB), B=@n16.A))
  n18 = NamedRerouteUsage()
  NamedRerouteDeclaration(None=ComponentMask(None=LinearInterpolate(A=@n14, B=@n17, Alpha=@n18)) {b=true, g=true, r=true}) {name=OutputBasecolor, nodeColor=<rgba 0.5 0 1 1>}
  NamedRerouteDeclaration(None=@n13.A) {name=Biome_Boreal_Mountain, nodeColor=<rgba 0.5 0 1 1>}
  NamedRerouteDeclaration(None=@n13.B) {name=Biome_Boreal_Tundra, nodeColor=<rgba 0.5 0 1 1>}
  n19 = GetMaterialAttributes(None=FunctionInput(Preview=SetMaterialAttributes!(MaterialAttributes)) {bUsePreviewValueAsDefault=true, inputName=Material, inputType=FunctionInput_MaterialAttributes}) {attributeGetTypes [#1, #2]}
  NamedRerouteDeclaration(None=Desaturation(None=NamedRerouteDeclaration(None=@n19.BaseColor) {name=Input_BaseColor, nodeColor=<rgba 0.5 0 1 1>}, Fraction=Constant() {r=0.5})) {name=BaseColorDesat, nodeColor=<rgba 0.5 0 1 1>}
  n20 = MaterialFunctionCall(Float4=FunctionInput() {bUsePreviewValueAsDefault=true, inputName=TerrainMasks_QuadD, inputType=FunctionInput_Vector4, sortPriority=40}) {materialFunction=&BreakOutFloat4Components}
  NamedRerouteDeclaration(None=@n20.R) {name=Biome_Hades, nodeColor=<rgba 0.5 0 1 1>}
  NamedRerouteDeclaration(None=@n20.G) {name=Biome_Olympus, nodeColor=<rgba 0.5 0 1 1>}
  NamedRerouteDeclaration(None=@n19.MaterialAttributes) {name=InputMaterial, nodeColor=<rgba 0.5 0 1 1>}
  NamedRerouteDeclaration(None=@n19.OpacityMask) {name=Input_OpacityMask, nodeColor=<rgba 0.5 0 1 1>}
 ]
}
