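# Landscape material function that tints gravel base color per biome and writes it back into the incoming material attributes.
# Biome masks unpacked from the TerrainMasks quad inputs drive a lerp chain of BiomeColors parameter tints, some applied over desaturated color.
# Boreal tundra also blends toward a high-elevation color through a powered height mask.
# Arid and open desert tints use ShadingPathSwitch for separate mobile colors; TerrainMasks_QuadA and Gravel_BiomeDesat are unconnected.
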MaterialFunction /Game/Athena/Helios/Environments/Landscape/Materials/MaterialFunctions_Utilities/MF_Gravel_BiomeColorTint
graph {
 properties {libraryCategoriesText [FN-Landscape]}
 nodes [
  Constant3Vector() {constant=<rgba 1 0 0 0>}
  FunctionInput() {bUsePreviewValueAsDefault=true, inputName=TerrainMasks_QuadA, inputType=FunctionInput_Vector4, sortPriority=10}
  FunctionOutput(None=SetMaterialAttributes!(MaterialAttributes=NamedRerouteUsage(), BaseColor=NamedRerouteUsage()) {attributeSetTypes [#1]})
  n1 = LinearInterpolate(A=Constant3Vector() {constant=<rgba 1 1 1 0>}, B=Constant3Vector() {constant=<rgba 0 0 1 0>}, Alpha=NamedRerouteUsage())
  n2 = Constant3Vector() {constant=<rgba 0 1 0 0>}
  n3 = NamedRerouteUsage()
  LinearInterpolate(
   A=LinearInterpolate(A=LinearInterpolate(A=@n1, B=@n2, Alpha=@n3), B=Constant3Vector() {constant=<rgba 0.7099 0 0.724 1>}, Alpha=NamedRerouteUsage())
   B=Constant3Vector() {constant=<rgba 1 1 0 1>}
   Alpha=NamedRerouteUsage()
  )
  LinearInterpolate(B=Constant3Vector())
  n4 = MaterialFunctionCall(Float4=FunctionInput() {bUsePreviewValueAsDefault=true, inputName=TerrainMasks_QuadB, inputType=FunctionInput_Vector4, sortPriority=20}) {materialFunction=&BreakOutFloat4Components}
  NamedRerouteDeclaration(None=@n4.R) {name=Biome_Chaparral_Coast, nodeColor=<rgba 0.5 0 1 1>}
  NamedRerouteDeclaration(None=@n4.G) {name=Biome_Boreal_Forest, nodeColor=<rgba 0.5 0 1 1>}
  n5 = NamedRerouteUsage()
  n6 = VectorParameter() {group=BiomeColors, parameterName=GravelColor_Boreal_Forest, defaultValue=<rgba 0.1812 0.1441 0.1144 1>}
  n7 = LinearInterpolate(A=Saturate(None=Multiply(A=@n5, B=@n6.RGB)), B=NamedRerouteUsage(), Alpha=NamedRerouteUsage())
  n8 = NamedRerouteUsage()
  n9 = VectorParameter() {group=BiomeColors, parameterName=GravelColor_Chaparral_Coast, defaultValue=<rgba 1 0.9837 0.8 1.4>}
  n10 = Multiply(A=Multiply(A=NamedRerouteUsage(), B=@n9.RGB), B=@n9.A)
  n11 = NamedRerouteUsage()
  n12 = VectorParameter() {group=BiomeColors, parameterName=GravelColor_Boreal_Tundra, defaultValue=<rgba 0.2552 0.203 0.1612 1>}
  n13 = VectorParameter() {group=BiomeColors, parameterName=GravelColor_Boreal_HeightmapHigh, defaultValue=<rgba 0.3 0.2363 0.21 1>}
  n14 = Multiply(A=@n13.RGB, B=Desaturation(None=NamedRerouteUsage(), Fraction=Constant() {r=0.9}))
  n15 = Multiply(
   A=ScalarParameter() {defaultValue=3, group=BiomeColors, parameterName=GravelColor_Boreal_Height_Mod}
   B=Power(Base=NamedRerouteUsage(), Exp=ScalarParameter() {defaultValue=10, group=BiomeColors, parameterName=GravelColor_Boreal_Height_Pow})
  )
  n16 = LinearInterpolate(
   A=Saturate(None=Multiply(A=@n11, B=@n12.RGB))
   B=NamedRerouteDeclaration(None=Saturate(None=@n14)) {name=BorealHeightColor, nodeColor=<rgba 0.5 0 1 1>}
   Alpha=NamedRerouteDeclaration(None=Saturate(None=@n15)) {name=BorealHeightMask, nodeColor=<rgba 0.5 0 1 1>}
  )
  n17 = LinearInterpolate(A=LinearInterpolate(A=NamedRerouteUsage(), B=Saturate(None=@n10), Alpha=NamedRerouteUsage()), B=@n16, Alpha=NamedRerouteUsage())
  n18 = VectorParameter() {group=BiomeColors, parameterName=GravelColor_Boreal_Mountain, defaultValue=<rgba 0.1812 0.1441 0.1144 1>}
  n19 = Saturate(None=Multiply(A=NamedRerouteUsage(), B=@n18.RGB))
  n20 = NamedRerouteUsage()
  n21 = NamedRerouteUsage()
  n22 = Desaturation(None=NamedRerouteUsage(), Fraction=Constant() {r=1})
  n23 = VectorParameter() {group=BiomeColors, parameterName=GravelColor_Hades, defaultValue=<rgba 0.4167 0.399 0.3817 1>}
  n24 = LinearInterpolate(
   A=LinearInterpolate(A=LinearInterpolate(A=@n17, B=@n7, Alpha=@n8), B=LinearInterpolate(A=@n19, B=@n20, Alpha=@n21), Alpha=NamedRerouteUsage())
   B=Saturate(None=Multiply(A=@n22, B=@n23.RGB))
   Alpha=NamedRerouteUsage()
  )
  n25 = NamedRerouteUsage()
  n26 = VectorParameter() {group=BiomeColors, parameterName=GravelColor_Olympus, defaultValue=<rgba 1 0.9837 0.8 1.4>}
  n27 = Saturate(None=Multiply(A=Multiply(A=@n25, B=@n26.RGB), B=@n26.A))
  n28 = NamedRerouteUsage()
  n29 = NamedRerouteUsage()
  n30 = Saturate(None=Multiply(A=@n29, B=VectorParameter() {group=BiomeColors, parameterName=GravelColor_Arid_Desert, defaultValue=<rgba 1.5 0.7154 0.375 1>}.RGB))
  n31 = LinearInterpolate(A=@n29, B=NamedRerouteUsage())
  n32 = VectorParameter() {
   group=BiomeColors
   parameterName=Mobile AridDesertGravel Color
   channelNames {a=BasecolorDesaturate, b="", g="", r=""}
   defaultValue=<rgba 1.2 0.6753 0.5794 1>
  }
  n33 = Multiply(A=@n31, B=@n32)
  n34 = VectorParameter() {
   group=BiomeColors
   parameterName=GravelColor_Open_Desert
   channelNames {a=BasecolorDesaturate, b="", g="", r=""}
   defaultValue=<rgba 2.25 0.9667 0.7875 0.25>
  }
  n35 = Multiply(A=Desaturation(None=NamedRerouteUsage(), Fraction=@n34.BasecolorDesaturate), B=@n34)
  n36 = VectorParameter() {
   group=BiomeColors
   parameterName=Mobile OpenDesertGravel Color
   channelNames {a=BasecolorDesaturate, b="", g="", r=""}
   defaultValue=<rgba 2 0.8365 0.677 1>
  }
  n37 = LinearInterpolate(
   A=LinearInterpolate(A=LinearInterpolate(A=@n24, B=@n27, Alpha=@n28), B=ShadingPathSwitch(Default=@n30, Mobile=@n33), Alpha=NamedRerouteUsage())
   B=ShadingPathSwitch(Default=Saturate(None=@n35), Mobile=Multiply(A=@n31, B=@n36))
   Alpha=NamedRerouteUsage()
  )
  n38 = NamedRerouteUsage()
  NamedRerouteDeclaration(None=ComponentMask(None=LinearInterpolate(A=@n37, B=@n16, Alpha=@n38)) {b=true, g=true, r=true}) {name=OutputBasecolor, nodeColor=<rgba 0.5 0 1 1>}
  NamedRerouteDeclaration(None=MaterialFunctionCall(Float4=FunctionInput() {bUsePreviewValueAsDefault=true, inputName=TerrainMasks_QuadC, inputType=FunctionInput_Vector4, sortPriority=30}) {materialFunction=&BreakOutFloat4Components}.G) {name=Biome_Doomstadt, nodeColor=<rgba 0.5 0 1 1>}
  NamedRerouteDeclaration(None=@n4.A) {name=Biome_Boreal_Mountain, nodeColor=<rgba 0.5 0 1 1>}
  NamedRerouteDeclaration(None=@n4.B) {name=Biome_Boreal_Tundra, nodeColor=<rgba 0.5 0 1 1>}
  n39 = GetMaterialAttributes(None=FunctionInput(Preview=SetMaterialAttributes!(MaterialAttributes)) {bUsePreviewValueAsDefault=true, inputName=Material, inputType=FunctionInput_MaterialAttributes}) {attributeGetTypes [#1, #2]}
  NamedRerouteDeclaration(None=Desaturation(None=NamedRerouteDeclaration(None=@n39.BaseColor) {name=Input_BaseColor, nodeColor=<rgba 0.5 0 1 1>}, Fraction=Constant() {r=0.5})) {name=BaseColorDesat, nodeColor=<rgba 0.5 0 1 1>}
  n40 = MaterialFunctionCall(Float4=FunctionInput() {bUsePreviewValueAsDefault=true, inputName=TerrainMasks_QuadD, inputType=FunctionInput_Vector4, sortPriority=40}) {materialFunction=&BreakOutFloat4Components}
  NamedRerouteDeclaration(None=@n40.R) {name=Biome_Hades, nodeColor=<rgba 0.5 0 1 1>}
  NamedRerouteDeclaration(None=@n40.G) {name=Biome_Olympus, nodeColor=<rgba 0.5 0 1 1>}
  NamedRerouteDeclaration(None=@n39.MaterialAttributes) {name=InputMaterial, nodeColor=<rgba 0.5 0 1 1>}
  NamedRerouteDeclaration(None=@n39.OpacityMask) {name=Input_OpacityMask, nodeColor=<rgba 0.5 0 1 1>}
  NamedRerouteDeclaration(None=@n40.B) {name=Biome_Arid_Desert, nodeColor=<rgba 0.5 0 1 1>}
  NamedRerouteDeclaration(None=@n40.A) {name=Biome_Open_Desert, nodeColor=<rgba 0.5 0 1 1>}
  NamedRerouteUsage()
  Reroute(None=@n37)
  ScalarParameter() {
   defaultValue=0.5
   desc=Saturation level of the color input from ChaparralGrassland before any of the biome color multipliers
   group=BiomeColors
   parameterName=Gravel_BiomeDesat
  }
 ]
}
