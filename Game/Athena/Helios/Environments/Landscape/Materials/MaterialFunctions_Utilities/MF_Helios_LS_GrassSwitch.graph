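# Material function for Helios landscapes that gates a grass weight and acts as an on/off switch for landscape grass.
# It multiplies the scalar input by a saturated world Z minus ZSpawnLimit, cutting grass below that height, then by the landscape visibility mask.
# A hard-coded static bool switches between this result and constant zero.
# The LS_Grass_MaskDepth parameter is present but unconnected.
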MaterialFunction /Game/Athena/Helios/Environments/Landscape/Materials/MaterialFunctions_Utilities/MF_Helios_LS_GrassSwitch
graph {nodes [
  n1 = WorldPosition()
  n2 = FunctionInput(Preview=Constant() {r=-1200}) {bUsePreviewValueAsDefault=true, description="Default= -1200", inputName=ZSpawnLimit, inputType=FunctionInput_Scalar, sortPriority=500}
  n3 = Multiply(
   A=FunctionInput(Preview=Constant()) {bUsePreviewValueAsDefault=true, description=Default is 0, inputType=FunctionInput_Scalar, sortPriority=-500}
   B=Saturate(None=Subtract(A=@n1.Z, B=@n2))
  )
  n4 = LandscapeVisibilityMask()
  FunctionOutput(None=StaticSwitch(True=Multiply(A=@n3, B=@n4), False=Constant(), Value=StaticBool() {desc=Set this to false to turn off landscape grass, value=true}))
  ScalarParameter() {defaultValue=-1200, parameterName=LS_Grass_MaskDepth}
 ]}
