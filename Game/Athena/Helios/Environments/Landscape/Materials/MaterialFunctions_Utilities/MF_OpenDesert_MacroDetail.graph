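# Material function for open-desert landscape macro detail, outputting wave normals, unmodified normals and four height-based sediment layer gradients.
# A tiled desert-wave normal map is scaled, transformed and masked by slope and BiomeMask.
# The same normals distort world Z, minus a parabolic periodic noise from world position, before per-layer height offsets and falloffs pack into SedimentLayers.
# Mobile uses plain world Z; the IsForStaticMesh switch and debug mask reach no output.
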MaterialFunction /Game/Athena/Helios/Environments/Landscape/Materials/MaterialFunctions_Utilities/MF_OpenDesert_MacroDetail
graph {nodes [
  n1 = WorldPosition()
  n2 = VectorParameter() {group=OpenDesert_MacroDetail, parameterName=OD_MD_DistortionDirection, defaultValue=<rgba -150 150 0 0>}
  n3 = TextureCoordinate()
  n4 = VectorParameter() {
   group=OpenDesert_MacroDetail
   parameterName=OD_MD_NormalSettings
   channelNames {a=Normal_SlopeMaskAdd, b=Normal_SlopeMaskMod, g=Normal_SlopeMaskPower, r=NormalScale}
   defaultValue=<rgba 0.02 50 0.35 0.15>
  }
  n5 = FunctionOutput(None=TextureSample(UVs=Multiply(A=@n3, B=@n4.NormalScale)) {samplerType=SAMPLERTYPE_Normal, texture=/Game/Environments/Helios/Terrain/Textures/T_Helios_DesertWaves_N}.RGB) {outputName=UnmodifiedNormals, sortPriority=-99}
  n6 = Multiply(A=@n2.RGB, B=@n5)
  n7 = Add(A=Add(A=@n1.Z, B=ComponentMask(None=@n6) {r=true}), B=ComponentMask(None=@n6) {g=true})
  n8 = Frac(None=Divide(A=@n1.XYZ, B=ScalarParameter() {defaultValue=3500, group=OpenDesert_MacroDetail, parameterName=OD_MD_SineWaveNoise_Scale, sortPriority=501}))
  n9 = Multiply(A=OneMinus(None=@n8), B=@n8)
  n10 = MaterialFunctionCall(Color=Multiply(A=@n9, B=@n9)) {materialFunction=&SplitComponents}
  n11 = Add(A=@n10.R, B=@n10.R)
  n12 = Subtract(
   A=@n7
   B=Multiply(
    A=Add(A=@n11, B=@n10.B)
    B=ScalarParameter() {defaultValue=1000, group=OpenDesert_MacroDetail, parameterName=OD_MD_SineWaveNoise_Mod, sortPriority=501}
   )
  )
  n13 = ShadingPathSwitch(Default=@n12, Mobile=@n1.Z)
  n14 = VectorParameter() {
   group=OpenDesert_MacroDetail
   parameterName=OD_MD_Layer1_2_Settings
   channelNames {a=Layer2_Falloff, b=Layer2_Height, g=Layer1_Falloff, r=Layer1_Height}
   defaultValue=<rgba -550 400 -900 450>
  }
  n15 = Add(A=@n13, B=@n14.Layer1_Height)
  n16 = Add(A=@n13, B=@n14.Layer2_Height)
  n17 = VectorParameter() {
   group=OpenDesert_MacroDetail
   parameterName=OD_MD_Layer3_4_Settings
   channelNames {a=Layer4_Falloff, b=Layer4_Height, g=Layer3_Falloff, r=Layer3_Height}
   defaultValue=<rgba -1250 350 -1500 600>
  }
  n18 = Add(A=@n13, B=@n17.Layer3_Height)
  n19 = Add(A=@n13, B=@n17.Layer4_Height)
  FunctionOutput(None=MaterialFunctionCall(
    X=Divide(A=@n15, B=@n14.Layer1_Falloff)
    Y=Divide(A=@n16, B=@n14.Layer2_Falloff)
    Z=Divide(A=@n18, B=@n17.Layer3_Falloff)
    A=Divide(A=@n19, B=@n17.Layer4_Falloff)
   ) {materialFunction=&MakeFloat4}.Result) {outputName=SedimentLayers, sortPriority=1}
  n20 = ComponentMask(None=Multiply(A=@n5, B=VectorParameter() {group=OpenDesert_MacroDetail, parameterName=OD_MD_NormalMod, defaultValue=<rgba 1 1 0.2 1>}.RGB)) {b=true, g=true, r=true}
  n21 = DotProduct(A=VertexNormalWS(), B=Constant3Vector() {constant=<rgba 0 0 1 0>})
  n22 = Saturate(None=Multiply(A=Power(Base=@n21, Exp=@n4.Normal_SlopeMaskPower), B=@n4.Normal_SlopeMaskMod))
  n23 = Multiply(A=Saturate(None=Add(A=@n22, B=@n4.Normal_SlopeMaskAdd)), B=FunctionInput() {inputName=BiomeMask, inputType=FunctionInput_Scalar})
  FunctionOutput(None=Multiply(A=Transform(None=@n20), B=@n23)) {outputName=Normals, sortPriority=-100}
  Max(A=ScalarParameter() {group=OpenDesert_MacroDetail, parameterName=OD_MD_BiomeMask_DEBUG}, B=@n23)
  ScalarParameter() {defaultValue=0.15, group=OpenDesert_MacroDetail, parameterName=OD_MD_Normal_SlopeMaskAdd, sortPriority=501}
  ScalarParameter() {defaultValue=400, group=OpenDesert_MacroDetail, parameterName=OD_MD_Layer1_Falloff, sortPriority=501}
  ScalarParameter() {defaultValue=-550, group=OpenDesert_MacroDetail, parameterName=OD_MD_Layer1_Height, sortPriority=501}
  ScalarParameter() {defaultValue=450, group=OpenDesert_MacroDetail, parameterName=OD_MD_Layer2_Falloff, sortPriority=501}
  ScalarParameter() {defaultValue=-900, group=OpenDesert_MacroDetail, parameterName=OD_MD_Layer2_Height, sortPriority=501}
  ScalarParameter() {defaultValue=350, group=OpenDesert_MacroDetail, parameterName=OD_MD_Layer3_Falloff, sortPriority=501}
  ScalarParameter() {defaultValue=-1250, group=OpenDesert_MacroDetail, parameterName=OD_MD_Layer3_Height, sortPriority=501}
  ScalarParameter() {defaultValue=600, group=OpenDesert_MacroDetail, parameterName=OD_MD_Layer4_Falloff, sortPriority=501}
  ScalarParameter() {defaultValue=-1500, group=OpenDesert_MacroDetail, parameterName=OD_MD_Layer4_Height, sortPriority=501}
  ScalarParameter() {defaultValue=0.02, group=OpenDesert_MacroDetail, parameterName=OD_MD_NormalScale}
  ScalarParameter() {defaultValue=50, group=OpenDesert_MacroDetail, parameterName=OD_MD_Normal_SlopeMaskPower, sortPriority=501}
  ScalarParameter() {defaultValue=0.35, group=OpenDesert_MacroDetail, parameterName=OD_MD_Normal_SlopeMaskMod, sortPriority=501}
  StaticSwitch(
   True=Subtract(A=@n7, B=@n8)
   False=@n12
   Value=FunctionInput(Preview=StaticBool()) {bUsePreviewValueAsDefault=true, inputName=IsForStaticMesh, inputType=FunctionInput_StaticBool, sortPriority=500}
  )
 ]}
