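# Material function that unpacks four packed terrain mask inputs (QuadA–D) into named biome and feature masks for the Helios landscape.
# QuadA and QuadC supply water, cliff, salt lake, river and Doomstadt masks, mostly as raw channels, with Hades water boosted and squared.
# QuadB and QuadD yield biome masks, using inverse-mask multiplies to reduce overlaps.
# Chaparral and Boreal biomes are also masked out by Hades and QuadD blue, with Grassland derived by inversion.
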MaterialFunction /Game/Athena/Helios/Environments/Landscape/Materials/MaterialFunctions_Utilities/MF_TerrainMasksIsolated
graph {nodes [
  n1 = FunctionInput() {bUsePreviewValueAsDefault=true, inputName=TerrainMasks_QuadA, inputType=FunctionInput_Vector4, sortPriority=-500, previewValue=<0 0 0 0>}
  n2 = MaterialFunctionCall(Float4=NamedRerouteDeclaration(None=@n1) {name=TerrainMasks_QuadA, nodeColor=<rgba 0.5 0 1 1>}) {materialFunction=&BreakOutFloat4Components}
  n3 = Saturate(None=Multiply(A=Constant() {r=2}, B=@n2.R))
  FunctionOutput(None=NamedRerouteDeclaration(None=Multiply(A=@n3, B=@n3)) {name=HadesWaterMask, nodeColor=<rgba 0.5 0 1 1>}) {outputName=01_HadesWaterMask, sortPriority=-1000}
  FunctionOutput(None=NamedRerouteDeclaration(None=@n2.G) {name=OasisCliffMask, nodeColor=<rgba 0.5 0 1 1>}) {outputName=02_OasisCliffMask, sortPriority=-990}
  n4 = FunctionInput() {bUsePreviewValueAsDefault=true, inputName=TerrainMasks_QuadC, inputType=FunctionInput_Vector4, sortPriority=-300, previewValue=<0 0 0 0>}
  n5 = MaterialFunctionCall(Float4=NamedRerouteDeclaration(None=@n4) {name=TerrainMasks_QuadC, nodeColor=<rgba 0.5 0 1 1>}) {materialFunction=&BreakOutFloat4Components}
  FunctionOutput(None=@n5.G) {outputName=11_Doomstadt, sortPriority=-900}
  FunctionOutput(None=@n5.B) {outputName=12_None, sortPriority=-890}
  FunctionOutput(None=@n5.A) {outputName=13_DoomstadtLeaves, sortPriority=-880}
  n6 = FunctionInput() {bUsePreviewValueAsDefault=true, inputName=TerrainMasks_QuadD, inputType=FunctionInput_Vector4, sortPriority=-200, previewValue=<0 0 0 0>}
  n7 = MaterialFunctionCall(Float4=NamedRerouteDeclaration(None=@n6) {name=TerrainMasks_QuadD, nodeColor=<rgba 0.5 0 1 1>}) {materialFunction=&BreakOutFloat4Components}
  FunctionOutput(None=NamedRerouteDeclaration(None=@n7.R) {name=Biome_Hades, desc=Hades, nodeColor=<rgba 0.5 0 1 1>}) {outputName=14_Biome_Hades, sortPriority=-870}
  n8 = Max(A=@n7.B, B=@n7.B)
  n9 = OneMinus(None=@n8)
  FunctionOutput(None=NamedRerouteDeclaration(None=Multiply(A=@n9, B=@n7.G)) {name=Biome_Olympus, desc=Olympus, nodeColor=<rgba 0.5 0 1 1>}) {outputName=15_Biome_Olympus, sortPriority=-860}
  n10 = OneMinus(None=@n7.A)
  FunctionOutput(None=NamedRerouteDeclaration(None=Multiply(A=@n7.B, B=@n10)) {name=Biome_Arid_Desert, desc=Arid_Desert, nodeColor=<rgba 0.5 0 1 1>}) {outputName=16_Biome_Arid_Desert, sortPriority=-850}
  FunctionOutput(None=NamedRerouteDeclaration(None=@n7.A) {name=Biome_Open_Desert, desc=Open_Desert, nodeColor=<rgba 0.5 0 1 1>}) {outputName=14_Biome_Open_Desert, sortPriority=-840}
  FunctionOutput(None=NamedRerouteDeclaration(None=@n2.B) {name=DesertDuneCliffMask, nodeColor=<rgba 0.5 0 1 1>}) {outputName=03_DesertDuneCliffMask, sortPriority=-980}
  FunctionOutput(None=@n2.A) {outputName=04_DesertSaltLake, sortPriority=-970}
  n11 = FunctionInput() {bUsePreviewValueAsDefault=true, inputName=TerrainMasks_QuadB, inputType=FunctionInput_Vector4, sortPriority=-400, previewValue=<0 0 0 0>}
  n12 = MaterialFunctionCall(Float4=NamedRerouteDeclaration(None=@n11) {name=TerrainMasks_QuadB, nodeColor=<rgba 0.5 0 1 1>}) {materialFunction=&BreakOutFloat4Components}
  n13 = Multiply(A=OneMinus(None=Max(A=@n12.G, B=@n12.G)), B=@n12.R)
  n14 = OneMinus(None=Max(A=Max(A=@n7.R, B=@n7.R), B=@n8))
  FunctionOutput(None=NamedRerouteDeclaration(None=Multiply(A=@n13, B=@n14)) {name=Biome_Chaparral_Coast, desc=Acronym:CHPC, nodeColor=<rgba 0.5 0 1 1>}) {outputName=05_Biome_Chaparral_Coast, sortPriority=-960}
  n15 = Multiply(A=@n12.G, B=OneMinus(None=@n12.A))
  FunctionOutput(None=NamedRerouteDeclaration(None=Multiply(A=@n15, B=@n14)) {name=Biome_Boreal_Forest, desc=Acronym:BRLF, nodeColor=<rgba 0.5 0 1 1>}) {outputName=06_Biome_Boreal_Forest, sortPriority=-950}
  n16 = Multiply(A=OneMinus(None=Max(A=@n12.G, B=@n12.G)), B=@n12.B)
  FunctionOutput(None=NamedRerouteDeclaration(None=Multiply(A=@n16, B=@n14)) {name=Biome_Boreal_Tundra, desc=Acronym:BRLT, nodeColor=<rgba 0.5 0 1 1>}) {outputName=07_Biome_Boreal_Tundra, sortPriority=-940}
  FunctionOutput(None=NamedRerouteDeclaration(None=Multiply(A=@n12.A, B=@n14)) {name=Biome_Boreal_Mountain, desc=Acronym:BRLM, nodeColor=<rgba 0.5 0 1 1>}) {outputName=08_Biome_Boreal_Mountain, sortPriority=-930}
  n17 = OneMinus(None=Max(A=Max(A=@n12.R, B=@n12.R), B=Max(A=@n12.B, B=@n12.B)))
  FunctionOutput(None=NamedRerouteDeclaration(None=Multiply(A=@n17, B=@n14)) {name=Biome_Chaparral_Grassland, desc=Acronym:CHPG, nodeColor=<rgba 0.5 0 1 1>}) {outputName=09_Biome_Chaparral_Grassland, sortPriority=-920}
  FunctionOutput(None=@n5.R) {outputName=10_DesertRiver, sortPriority=-910}
  n18 = Multiply() {desc=G}
  n19 = Multiply() {desc=B}
  Multiply(A=OneMinus(None=@n18), B=@n19) {desc=Soft transition}
  Subtract(A=@n19, B=@n18) {desc=Hard transition}
 ]}
