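# Material function that supplies world-space variation masks for the Helios landscape materials on mobile.
# It truncates world position XY from LWC, scales it by a small constant and uses it as UVs into the mobile Helios noise texture.
# A single masks-sampler texture fetch outputs all four RGBA channels, giving a large-scale noise pattern independent of mesh UVs.
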MaterialFunction /Game/Athena/Helios/Environments/Landscape/Materials/MaterialFunctions_Utilities/MF_VariationMasks_Mobile
graph {nodes [
  n1 = TruncateLWC!(None=WorldPosition().XY)
  FunctionOutput(None=TextureSample(UVs=Multiply(A=@n1) {constB=0.0001}) {samplerType=SAMPLERTYPE_Masks, texture=/Game/Environments/Helios/Terrain/TexturesMobile/T_Helios_Mobile_Noise}.RGBA) {outputName=VariationMasks Mobile}
 ]}
