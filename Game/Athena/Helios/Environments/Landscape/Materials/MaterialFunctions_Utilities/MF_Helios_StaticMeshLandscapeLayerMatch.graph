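# Material function that rebuilds Helios landscape layers (grasses, desert oasis, gravel, sand, asphalt, snow, scree, brimstone) for static meshes to match the terrain, with world-space normals.
# Shared terrain, biome and mobile variation masks feed each layer function, whose StaticMeshMaterial becomes a separate output, mostly at full weight.
# Sand adds open-desert macro detail, and grass accepts a custom slope mask.
# Extra outputs expose the oasis cliff mask and snow's post-RVT result.
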MaterialFunction /Game/Athena/Helios/Environments/Landscape/Materials/MaterialFunctions_Utilities/MF_Helios_StaticMeshLandscapeLayerMatch
graph {
 properties {description=Normals are in worldspace.}
 nodes [
  FunctionOutput(None=MaterialFunctionCall!(
    Material
    MobileMaterial
    GrassWeight
    UseTangentNormals=StaticBool()
    LayerWeight=NamedRerouteUsage()
    TerrainMasks_QuadA=NamedRerouteUsage()
    TerrainMasks_QuadB=NamedRerouteUsage()
    TerrainMasks_QuadC=NamedRerouteUsage()
    TerrainMasks_QuadD=NamedRerouteUsage()
    VariationMasks Mobile=NamedRerouteUsage()
    Chap_D
    Chap_S
    Chap_N
   ) {materialFunction=/Game/Athena/Helios/Environments/Landscape/Materials/LandscapeLayers/MF_Helios_Grass_ChaparralCoast}.StaticMeshMaterial) {outputName=Grass_ChaparralCoast, sortPriority=15}
  n1 = StaticBool()
  n2 = FunctionInput(Preview=StaticBool()) {
   bUsePreviewValueAsDefault=true
   desc=Setting this to true will ignore Grass_EdgeOnly input
   description=Setting this to true will ignore Grass_EdgeOnly input
   inputName=Grass_UseCustomSlopeMask
   inputType=FunctionInput_StaticBool
   sortPriority=13
  }
  n3 = FunctionInput(Preview=Constant()) {bUsePreviewValueAsDefault=true, inputName=Grass_CustomSlopeMask, inputType=FunctionInput_Scalar, sortPriority=14}
  n4 = Constant() {r=1}
  n5 = NamedRerouteUsage()
  n6 = NamedRerouteUsage()
  n7 = NamedRerouteUsage()
  n8 = NamedRerouteUsage()
  n9 = NamedRerouteUsage()
  FunctionOutput(None=NamedRerouteDeclaration(None=MaterialFunctionCall!(
     Material
     MobileMaterial
     IsForStaticMesh
     UseTangentNormals=@n1
     UseCustomSlopeMask=@n2
     CustomSlopeMask=@n3
     Grass_EdgeOnly
     LayerWeight=@n4
     TerrainMasks_QuadA=@n5
     TerrainMasks_QuadB=@n6
     TerrainMasks_QuadC=@n7
     TerrainMasks_QuadD=@n8
     VariationMasks Mobile=@n9
    ) {materialFunction=/Game/Athena/Helios/Environments/Landscape/Materials/LandscapeLayers/MF_Helios_Grass}.StaticMeshMaterial) {name=GrassLayer, nodeColor=<rgba 0.5 0 1 1>}) {outputName=GrassLayer, sortPriority=10}
  n10 = MaterialFunctionCall!(UseLandscapeCoords, "UseProjection(translucent)") {materialFunction=&/MF_TerrainMasks}
  n11 = MaterialFunctionCall!(
   TerrainMasks_QuadA=NamedRerouteDeclaration(None=@n10.QuadA_RGBA) {name=TerrainMasks_QuadA, nodeColor=<rgba 0.5 0 1 1>}
   TerrainMasks_QuadB=NamedRerouteDeclaration(None=@n10.QuadB_RGBA) {name=TerrainMasks_QuadB, nodeColor=<rgba 0.5 0 1 1>}
   TerrainMasks_QuadC=NamedRerouteDeclaration(None=@n10.QuadC_RGBA) {name=TerrainMasks_QuadC, nodeColor=<rgba 0.5 0 1 1>}
   TerrainMasks_QuadD=NamedRerouteDeclaration(None=@n10.QuadD_RGBA) {name=TerrainMasks_QuadD, nodeColor=<rgba 0.5 0 1 1>}
  ) {materialFunction=&/MF_TerrainMasksIsolated}
  FunctionOutput(None=NamedRerouteDeclaration(None=@n11.02_OasisCliffMask) {name=OasisCliffMask, nodeColor=<rgba 0.5 0 1 1>}) {outputName=OasisCliffMask, sortPriority=9999}
  FunctionOutput(None=MaterialFunctionCall!(
    Material
    MobileMaterial
    MobileGrass
    LayerWeight=Constant() {r=1}
    UseTangentNormals
    TerrainMasks_QuadA=NamedRerouteUsage()
    TerrainMasks_QuadB=NamedRerouteUsage()
    TerrainMasks_QuadC=NamedRerouteUsage()
    TerrainMasks_QuadD=NamedRerouteUsage()
    VariationMasks Mobile=NamedRerouteUsage()
   ) {materialFunction=/Game/Athena/Helios/Environments/Landscape/Materials/LandscapeLayers/MF_Helios_DesertOasis}.StaticMeshMaterial) {outputName=DesertOasisLayer, sortPriority=16}
  FunctionOutput(None=MaterialFunctionCall!(
    Material
    MobileMaterial
    IsForStaticMesh
    UseTangentNormals=StaticBool()
    LayerWeight=Constant() {r=1}
    MudWeight
    TerrainMasks_QuadA=NamedRerouteUsage()
    TerrainMasks_QuadB=NamedRerouteUsage()
    TerrainMasks_QuadC=NamedRerouteUsage()
    TerrainMasks_QuadD=NamedRerouteUsage()
    VariationMasks Mobile=NamedRerouteUsage()
   ) {materialFunction=/Game/Athena/Helios/Environments/Landscape/Materials/LandscapeLayers/MF_Helios_Gravel}.StaticMeshMaterial) {outputName=GravelLayer, sortPriority=20}
  n12 = MaterialFunctionCall!(BiomeMask=NamedRerouteUsage(), IsForStaticMesh=StaticBool() {value=true}) {materialFunction=&/MF_OpenDesert_MacroDetail}
  FunctionOutput(None=MaterialFunctionCall!(
    Material
    MobileMaterial
    LayerWeight=Constant() {r=1}
    UseTangentNormals=StaticBool()
    TerrainMasks_QuadA=NamedRerouteUsage()
    TerrainMasks_QuadB=NamedRerouteUsage()
    TerrainMasks_QuadC=NamedRerouteUsage()
    TerrainMasks_QuadD=NamedRerouteUsage()
    VariationMasks Mobile=NamedRerouteUsage()
    OD_MD_Normals=@n12.Normals
    OD_MD_SedimentLayers=@n12.Normals
   ) {materialFunction=/Game/Athena/Helios/Environments/Landscape/Materials/LandscapeLayers/MF_Helios_Sand}.StaticMeshMaterial) {outputName=SandLayer, sortPriority=40}
  FunctionOutput(None=MaterialFunctionCall!(
    Material
    MobileMaterial
    UseTangentNormals=StaticBool()
    LayerWeight=Constant() {r=1}
    TerrainMasks_QuadA=NamedRerouteUsage()
    TerrainMasks_QuadB=NamedRerouteUsage()
    TerrainMasks_QuadC=NamedRerouteUsage()
    TerrainMasks_QuadD=NamedRerouteUsage()
    VariationMasks Mobile=NamedRerouteUsage()
   ) {materialFunction=/Game/Athena/Helios/Environments/Landscape/Materials/LandscapeLayers/MF_Helios_Asphalt}.StaticMeshMaterial) {outputName=AsphaltLayer, sortPriority=50}
  n13 = MaterialFunctionCall!(
   Material
   MobileMaterial
   IsForStaticMesh
   UseTangentNormals=StaticBool()
   LayerWeight=Constant() {r=1}
   TerrainMasks_QuadA=NamedRerouteUsage()
   TerrainMasks_QuadB=NamedRerouteUsage()
   TerrainMasks_QuadC=NamedRerouteUsage()
   TerrainMasks_QuadD=NamedRerouteUsage()
   CliffMaterialOnly
   CliffLayerMask
   VariationMasks Mobile=NamedRerouteUsage()
  ) {materialFunction=/Game/Athena/Helios/Environments/Landscape/Materials/LandscapeLayers/MF_Helios_Snow_LSLayer}
  FunctionOutput(None=@n13.StaticMeshMaterial) {outputName=SnowLayer, sortPriority=300}
  FunctionOutput(None=@n13.PostRVT) {outputName=SnowDynamic, sortPriority=310}
  FunctionOutput(None=MaterialFunctionCall!(
    Material
    MobileMaterial
    UseTangentNormals=StaticBool()
    LayerWeight=Constant() {r=1}
    TerrainMasks_QuadA=NamedRerouteUsage()
    TerrainMasks_QuadB=NamedRerouteUsage()
    TerrainMasks_QuadC=NamedRerouteUsage()
    TerrainMasks_QuadD=NamedRerouteUsage()
    VariationMasks Mobile=NamedRerouteUsage()
   ) {materialFunction=/Game/Athena/Helios/Environments/Landscape/Materials/LandscapeLayers/MF_Helios_Scree}.StaticMeshMaterial) {outputName=ScreeLayer, sortPriority=60}
  FunctionOutput(None=MaterialFunctionCall!(
    Material
    MobileMaterial
    UseTangentNormals
    LayerWeight=Constant() {r=1}
    TerrainMasks_QuadA=NamedRerouteUsage()
    TerrainMasks_QuadB=NamedRerouteUsage()
    TerrainMasks_QuadC=NamedRerouteUsage()
    TerrainMasks_QuadD=NamedRerouteUsage()
    VariationMasks Mobile=NamedRerouteUsage()
   ) {materialFunction=/Game/Athena/Helios/Environments/Landscape/Materials/LandscapeLayers/MF_Helios_Brimstone01}.StaticMeshMaterial) {outputName=Brimstone Layer, sortPriority=16}
  FunctionOutput(None=MaterialFunctionCall!(
    Material=NamedRerouteUsage()
    MobileMaterial
    ChapCoastMobile
    GrassWeight=Constant() {r=1}
    LayerWeight=NamedRerouteUsage()
    UseTangentNormals=StaticBool()
    TerrainMasks_QuadA=NamedRerouteUsage()
    TerrainMasks_QuadB=NamedRerouteUsage()
    TerrainMasks_QuadC=NamedRerouteUsage()
    TerrainMasks_QuadD=NamedRerouteUsage()
    VariationMasks Mobile=NamedRerouteUsage()
   ) {materialFunction=/Game/Athena/Helios/Environments/Landscape/Materials/LandscapeLayers/MF_Helios_Grass_Desert}.StaticMeshMaterial) {outputName=DesertGrassLayer, sortPriority=16}
  NamedRerouteDeclaration(None=Max(
    A=NamedRerouteDeclaration(None=@n11.16_Biome_Arid_Desert) {name=Biome_Arid_Desert, desc=Arid_Desert, nodeColor=<rgba 0.5 0 1 1>}
    B=NamedRerouteDeclaration(None=@n11.14_Biome_Open_Desert) {name=Biome_Open_Desert, desc=Open_Desert, nodeColor=<rgba 0.5 0 1 1>}
   )) {name=Biome_Arid_PLUS_OPEN_Desert, nodeColor=<rgba 0.5 0 1 1>}
  NamedRerouteDeclaration(None=@n11.09_Biome_Chaparral_Grassland) {name=Biome_Chaparral_Grassland, desc=Acronym:CHPG, nodeColor=<rgba 0.5 0 1 1>}
  NamedRerouteDeclaration(None=MaterialFunctionCall() {materialFunction=&/MF_VariationMasks_Mobile}.VariationMasks Mobile) {name=VariationMasks Mobile, nodeColor=<rgba 1 0.9141 0 1>}
  NamedRerouteDeclaration(None=@n11.05_Biome_Chaparral_Coast) {name=Biome_Chaparral_Coast, desc=Acronym:CHPC, nodeColor=<rgba 0.5 0 1 1>}
  NamedRerouteDeclaration(None=@n11.06_Biome_Boreal_Forest) {name=Biome_Boreal_Forest, desc=Acronym:BRLF, nodeColor=<rgba 0.5 0 1 1>}
  NamedRerouteDeclaration(None=@n11.07_Biome_Boreal_Tundra) {name=Biome_Boreal_Tundra, desc=Acronym:BRLT, nodeColor=<rgba 0.5 0 1 1>}
  NamedRerouteDeclaration(None=@n11.08_Biome_Boreal_Mountain) {name=Biome_Boreal_Mountain, desc=Acronym:BRLM, nodeColor=<rgba 0.5 0 1 1>}
  NamedRerouteDeclaration(None=@n11.14_Biome_Hades) {name=Biome_Hades, desc=Hades, nodeColor=<rgba 0.5 0 1 1>}
  NamedRerouteDeclaration(None=@n11.15_Biome_Olympus) {name=Biome_Olympus, desc=Olympus, nodeColor=<rgba 0.5 0 1 1>}
 ]
}
