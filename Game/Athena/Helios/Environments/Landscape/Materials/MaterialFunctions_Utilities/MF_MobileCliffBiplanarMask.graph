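# Material function for mobile cliff materials that outputs a biplanar blend mask separating X-facing from Y-facing sides.
# It subtracts absolute Y from absolute X of the world-space vertex normal, scales by a Sharpness input (preview default 10), then offsets, halves and saturates the result.
# Using only the vertex normal keeps the mask free of texture samples, with Sharpness controlling the transition width between the two projections.
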MaterialFunction /Game/Athena/Helios/Environments/Landscape/Materials/MaterialFunctions_Utilities/MF_MobileCliffBiplanarMask
graph {nodes [
  n1 = Abs(None=ComponentMask(None=VertexNormalWS()) {g=true, r=true})
  n2 = ComponentMask(None=@n1) {r=true}
  n3 = ComponentMask(None=@n1) {g=true}
  n4 = FunctionInput() {bUsePreviewValueAsDefault=true, inputName=Sharpness, inputType=FunctionInput_Scalar, sortPriority=1, previewValue=<10 10 10 10>}
  n5 = Add(A=Multiply(A=Subtract(A=@n2, B=@n3), B=@n4) {constB=10})
  FunctionOutput(None=Saturate(None=Multiply(A=@n5) {constB=0.5}))
 ]}
