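# Material function for mobile grass that adds a view-dependent fresnel tint to the Base Color of incoming material attributes.
# The inverted, saturated dot of camera vector and pixel normal is raised to the eighth power by repeated multiplies rather than a Fresnel or Power node.
# A layer mask and a color, defaulting to the Mobile Grass Fresnel parameter, scale the term before it is added and output as Grass Fresnel.
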MaterialFunction /Game/Athena/Helios/Environments/Landscape/Materials/MaterialFunctions_Utilities/MF_MobileGrass_Fresnel
graph {nodes [
  n1 = GetMaterialAttributes(None=FunctionInput() {inputName=GrassMaterial, inputType=FunctionInput_MaterialAttributes}) {attributeGetTypes [#1]}
  n2 = CameraVectorWS()
  n3 = PixelNormalWS()
  n4 = OneMinus(None=Saturate(None=DotProduct(A=@n2, B=@n3)))
  n5 = Multiply(A=@n4, B=@n4)
  n6 = Multiply(A=@n5, B=@n5)
  n7 = FunctionInput() {bUsePreviewValueAsDefault=true, inputName=Layer Mask, inputType=FunctionInput_Scalar, previewValue=<1 1 1 1>}
  n8 = FunctionInput(Preview=VectorParameter() {group=Mobile Grass, parameterName=Mobile Grass Fresnel, defaultValue=<rgba 0.1305 0.1406 0.0403 1>}.RGB) {bUsePreviewValueAsDefault=true, inputName=Frensel Color, sortPriority=4, previewValue=<1 1 1 1>}
  n9 = Multiply(A=Multiply(A=Multiply(A=@n6, B=@n6), B=@n7), B=@n8)
  FunctionOutput(None=SetMaterialAttributes!(MaterialAttributes=@n1.MaterialAttributes, BaseColor=Add(A=@n1.BaseColor, B=@n9)) {attributeSetTypes [#1]}) {outputName=Grass Fresnel}
 ]}
