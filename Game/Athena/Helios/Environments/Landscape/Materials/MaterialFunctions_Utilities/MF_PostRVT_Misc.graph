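# Landscape material function for miscellaneous adjustments to material attributes, named as a post-RVT stage of the Helios terrain material.
# The incoming attributes and a mobile variation mask input pass through MF_Curie_Landscape, which handles fire.
# The result gets its Opacity Mask from the Landscape Visibility Mask for terrain holes.
# MF_Terrain_LumenModification, shared from Asteria, processes the attributes last before output.
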MaterialFunction /Game/Athena/Helios/Environments/Landscape/Materials/MaterialFunctions_Utilities/MF_PostRVT_Misc
graph {nodes [
  n1 = FunctionInput() {
   bUsePreviewValueAsDefault=true
   inputName=VariationMasksMobile
   inputType=FunctionInput_Vector4
   sortPriority=1
   previewValue=<0.7 0.7 0.7 0.7>
  }
  n2 = MaterialFunctionCall!(
   In=FunctionInput(Preview=GetMaterialAttributes().MaterialAttributes) {bUsePreviewValueAsDefault=true, inputName=LandscapeMaterial, inputType=FunctionInput_MaterialAttributes}
   UseHeightFalloff
   OnFireEmissiveOverride
   HeightFalloffRadius
   HeightFalloffHardness
   VariationMasksMobile=@n1
   In_Grass
   Is_Stylised_WE
  ) {desc=Fire, materialFunction=/Game/Packages/Fortress_SharedMaterials/Base_Material_Functions/MF_Curie_Landscape}
  n3 = LandscapeVisibilityMask()
  FunctionOutput(None=MaterialFunctionCall!(In=SetMaterialAttributes!(MaterialAttributes=@n2.Result, OpacityMask=@n3) {attributeSetTypes [#1]}) {materialFunction=/Game/Athena/Asteria/Environments/Landscape/Materials/MaterialFunctions/MF_Terrain_LumenModification}.Result) {outputName=LandscapeMaterial}
 ]}
