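# Material function that outputs eight variation masks for Helios landscape materials by sampling the shared T_Helios_Landscape_Noise texture at different scales.
# Coordinates, defaulting to world XY, are divided by per-mask scale inputs to form UVs for four red-channel and four green-channel samples.
# The third mask of each channel adds a fixed coordinate offset, and two green masks default to negative scales, mirroring the noise.
# Matching VariationMask scalar parameters sit unconnected in the graph.
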MaterialFunction /Game/Athena/Helios/Environments/Landscape/Materials/MaterialFunctions_Utilities/MF_VariationMasks2
graph {nodes [
  n1 = NamedRerouteUsage()
  n2 = FunctionInput(Preview=Constant() {r=5000}) {bUsePreviewValueAsDefault=true, inputName=R_01_Scale, inputType=FunctionInput_Scalar, sortPriority=10}
  FunctionOutput(None=TextureSample(UVs=Divide(A=@n1, B=@n2), Tex=NamedRerouteUsage()) {samplerSource=SSM_Wrap_WorldGroupSettings, texture=/Game/Environments/Helios/Terrain/Textures/T_Helios_Landscape_Noise}.R) {outputName=R_01, sortPriority=10}
  n3 = NamedRerouteUsage()
  n4 = FunctionInput(Preview=Constant() {r=12000}) {bUsePreviewValueAsDefault=true, inputName=R_02_Scale, inputType=FunctionInput_Scalar, sortPriority=20}
  FunctionOutput(None=TextureSample(UVs=Divide(A=@n3, B=@n4), Tex=NamedRerouteUsage()) {samplerSource=SSM_Wrap_WorldGroupSettings, texture=&T_Helios_Landscape_Noise}.R) {outputName=R_02, sortPriority=20}
  n5 = Add(A=NamedRerouteUsage(), B=Constant() {r=500})
  n6 = FunctionInput(Preview=Constant() {r=10000}) {bUsePreviewValueAsDefault=true, inputName=R_03_Scale, inputType=FunctionInput_Scalar, sortPriority=30}
  FunctionOutput(None=TextureSample(UVs=Divide(A=@n5, B=@n6), Tex=NamedRerouteUsage()) {samplerSource=SSM_Wrap_WorldGroupSettings, texture=&T_Helios_Landscape_Noise}.R) {outputName=R_03, sortPriority=30}
  n7 = NamedRerouteUsage()
  n8 = FunctionInput(Preview=Constant() {r=25000}) {bUsePreviewValueAsDefault=true, inputName=R_04_Scale, inputType=FunctionInput_Scalar, sortPriority=40}
  FunctionOutput(None=TextureSample(UVs=Divide(A=@n7, B=@n8), Tex=NamedRerouteUsage()) {samplerSource=SSM_Wrap_WorldGroupSettings, texture=&T_Helios_Landscape_Noise}.R) {outputName=R_04, sortPriority=40}
  n9 = NamedRerouteUsage()
  n10 = FunctionInput(Preview=Constant() {r=5599}) {bUsePreviewValueAsDefault=true, inputName=G_01_Scale, inputType=FunctionInput_Scalar, sortPriority=50}
  FunctionOutput(None=TextureSample(UVs=Divide(A=@n9, B=@n10), Tex=NamedRerouteUsage()) {samplerSource=SSM_Wrap_WorldGroupSettings, texture=&T_Helios_Landscape_Noise}.G) {outputName=G_01, sortPriority=50}
  n11 = NamedRerouteUsage()
  n12 = FunctionInput(Preview=Constant() {r=-3600}) {bUsePreviewValueAsDefault=true, inputName=G_02_Scale, inputType=FunctionInput_Scalar, sortPriority=60}
  FunctionOutput(None=TextureSample(UVs=Divide(A=@n11, B=@n12), Tex=NamedRerouteUsage()) {samplerSource=SSM_Wrap_WorldGroupSettings, texture=&T_Helios_Landscape_Noise}.G) {outputName=G_02, sortPriority=60}
  n13 = Add(A=NamedRerouteUsage(), B=Constant() {r=500})
  n14 = FunctionInput(Preview=Constant() {r=4560}) {bUsePreviewValueAsDefault=true, inputName=G_03_Scale, inputType=FunctionInput_Scalar, sortPriority=70}
  FunctionOutput(None=TextureSample(UVs=Divide(A=@n13, B=@n14), Tex=NamedRerouteUsage()) {samplerSource=SSM_Wrap_WorldGroupSettings, texture=&T_Helios_Landscape_Noise}.G) {outputName=G_03, sortPriority=70}
  n15 = NamedRerouteUsage()
  n16 = FunctionInput(Preview=Constant() {r=-13000}) {bUsePreviewValueAsDefault=true, inputName=G_04_Scale, inputType=FunctionInput_Scalar, sortPriority=80}
  FunctionOutput(None=TextureSample(UVs=Divide(A=@n15, B=@n16), Tex=NamedRerouteUsage()) {samplerSource=SSM_Wrap_WorldGroupSettings, texture=&T_Helios_Landscape_Noise}.G) {outputName=G_04, sortPriority=80}
  FunctionOutput(None=Constant()) {outputName=---------}
  NamedRerouteDeclaration(None=TextureObject() {texture=&T_Helios_Landscape_Noise}) {name=NoiseTexture, nodeColor=<rgba 0 1 0.5547 1>}
  n17 = FunctionInput(Preview=WorldPosition().XY) {
   bUsePreviewValueAsDefault=true
   description="Default=XYWorldPos(ExcludingMaterialOffsets)"
   inputName=Coordinates
   inputType=FunctionInput_Vector2
   sortPriority=-10
  }
  NamedRerouteDeclaration(None=@n17) {name=XYWorldPos, nodeColor=<rgba 0 1 0.5547 1>}
  ScalarParameter() {defaultValue=12000, group=VariationMask, parameterName=R_02_Scale, sortPriority=-45}
  ScalarParameter() {defaultValue=5000, group=VariationMask, parameterName=R_01_Scale, sortPriority=-50}
  ScalarParameter() {defaultValue=10000, group=VariationMask, parameterName=R_03_Scale, sortPriority=-40}
  ScalarParameter() {defaultValue=25000, group=VariationMask, parameterName=R_04_Scale, sortPriority=-30}
  ScalarParameter() {defaultValue=-3600, group=VariationMask, parameterName=G_02_Scale, sortPriority=-45}
  ScalarParameter() {defaultValue=5599, group=VariationMask, parameterName=G_01_Scale, sortPriority=-50}
  ScalarParameter() {defaultValue=4560, group=VariationMask, parameterName=G_03_Scale, sortPriority=-40}
  ScalarParameter() {defaultValue=-13000, group=VariationMask, parameterName=G_04_Scale, sortPriority=-30}
 ]}
