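# Helios landscape material function that supplies noise masks and a flower color, apparently for coloring Scotch broom.
# The layer noise mask blends two UV scales of landscape noise through BeersLaw as a contrast shaper.
# The flower mask sharpens an inverted MF_VariationMasks2 channel with a power, alongside a parameter-driven color.
# A debug output exposes scaled, biased world-space noise, and several experimental branches and parameters remain unconnected.
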MaterialFunction /Game/Athena/Helios/Environments/Landscape/Materials/MaterialFunctions_Utilities/MF_ScotchBroom_Colors
graph {nodes [
  Constant() {r=4000}
  Constant() {r=2300}
  Constant() {r=1000}
  n1 = Multiply(A=TextureCoordinate(), B=Constant() {r=0.03})
  n2 = TextureObject() {texture=/Game/Environments/Helios/Terrain/Textures/T_Helios_Landscape_Noise}
  n3 = TextureSample(UVs=@n1, Tex=@n2) {samplerSource=SSM_Wrap_WorldGroupSettings, texture=/Game/Global/Textures/Default/Noise/T_Tiling_Noise_01}
  n4 = OneMinus(None=@n3.B)
  n5 = Multiply(A=@n1, B=Constant() {r=6})
  n6 = SquareRoot(None=Multiply(
    A=@n3.B
    B=TextureSample(UVs=@n5, Tex=@n2) {samplerSource=SSM_Wrap_WorldGroupSettings, texture=&T_Tiling_Noise_01}.B
   ))
  n7 = Constant() {r=3}
  n8 = OneMinus(None=Multiply(
    A=@n4
    B=MaterialFunctionCall!(Thickness=@n6, Depth Scale=@n7) {materialFunction=/Engine/Functions/Engine_MaterialFunctions01/Volumetrics/BeersLaw}.Result
   ))
  FunctionOutput(None=@n8) {outputName=LandscapeLayerMaskNoise}
  n9 = MaterialFunctionCall!(
   Coordinates
   R_01_Scale=ScalarParameter() {defaultValue=4000, group=Gravel, parameterName=ChaparralCoast_Noise1_Scale}
   R_02_Scale=ScalarParameter() {defaultValue=2300, group=Gravel, parameterName=ChaparralCoast_Noise2_Scale}
   R_03_Scale=ScalarParameter() {defaultValue=2300, group=Gravel, parameterName=ChaparralCoast_Noise3_Scale}
   R_04_Scale
   G_01_Scale
   G_02_Scale
   G_03_Scale
   G_04_Scale
  ) {materialFunction=&/MF_VariationMasks2}
  n10 = Power(Base=OneMinus(None=NamedRerouteDeclaration(None=@n9.R_02) {name=ChaparralCoast_Noise2, nodeColor=<rgba 0.5 0 1 1>}), Exp=Constant() {r=10}) {constExponent=4}
  FunctionOutput(None=@n10) {outputName=FlowerWorldMask1}
  FunctionOutput(None=VectorParameter() {parameterName=ChaparralGrassColor1, defaultValue=<rgba 0.7 0.4313 0.0608 1>}.RGB) {outputName=Flower1_Color}
  n11 = TextureSample(UVs=Divide(
    A=WorldPosition() {worldPositionShaderOffset=WPT_ExcludeAllShaderOffsets}.XY
    B=ScalarParameter() {defaultValue=5000, parameterName=debug_scale111}
   )) {samplerSource=SSM_Wrap_WorldGroupSettings, texture=&T_Helios_Landscape_Noise}
  n12 = Multiply(A=@n11.R, B=ScalarParameter() {defaultValue=2, parameterName=debug_222})
  n13 = ScalarParameter() {defaultValue=-0.25, parameterName=debug_333}
  FunctionOutput(None=Saturate(None=Add(A=@n12, B=@n13))) {outputName=DEBUGOUT, sortPriority=500}
  n14 = ScalarParameter() {defaultValue=4, group=Gravel, parameterName=debug_gnggg4k88pow}
  n15 = NamedRerouteUsage()
  LinearInterpolate(
   A=LinearInterpolate(A=Power(Base=@n4, Exp=@n14), B=@n8, Alpha=ScalarParameter() {defaultValue=1, group=Gravel, parameterName=debug_gnggg4k88})
   B=Power(Base=OneMinus(None=@n15), Exp=ScalarParameter() {defaultValue=2, group=Gravel, parameterName=ChaparralCoast_Noise1_Pow}) {constExponent=4}
   Alpha=ScalarParameter() {group=Gravel, parameterName=debug_gnois343434fq}
  )
  NamedRerouteDeclaration(None=@n9.R_03) {name=ChaparralCoast_Noise3, nodeColor=<rgba 0.5 0 1 1>}
  NamedRerouteDeclaration(None=@n9.R_01) {name=ChaparralCoast_Noise1, nodeColor=<rgba 0.5 0 1 1>}
  OneMinus(None=@n11.B)
  Power(Base=OneMinus(None=@n11.B), Exp=ScalarParameter() {defaultValue=15, parameterName=debug_111}) {constExponent=4}
  n16 = WorldPosition() {worldPositionShaderOffset=WPT_ExcludeAllShaderOffsets}
  n17 = Constant() {r=7000}
  Power(
   Base=TextureSample(UVs=Divide(A=@n16.XY, B=@n17)) {samplerSource=SSM_Wrap_WorldGroupSettings, texture=&T_Helios_Landscape_Noise}.R
   Exp=Constant() {r=2}
  ) {constExponent=4}
  Saturate(None=Multiply(A=@n10, B=Constant() {r=1}))
  n18 = TextureSample(UVs=Divide(A=WorldPosition() {worldPositionShaderOffset=WPT_ExcludeAllShaderOffsets}.XY, B=Constant() {r=10000})) {samplerSource=SSM_Wrap_WorldGroupSettings, texture=&T_Helios_Landscape_Noise}
  n19 = Constant() {r=5}
  Saturate(None=Add(A=Multiply(A=@n18.G, B=@n19), B=Constant() {r=-0.9}))
  ScalarParameter() {defaultValue=10, group=Gravel, parameterName=ChaparralCoast_Noise3_Pow}
  ScalarParameter() {defaultValue=1, group=Gravel, parameterName=ChaparralCoast_Noise3_Mod}
  ScalarParameter() {defaultValue=3, group=Gravel, parameterName=debug_gnoisbeers}
  ScalarParameter() {defaultValue=0.03, group=Gravel, parameterName=debug_gnoisggggt}
  ScalarParameter() {defaultValue=6, group=Gravel, parameterName=debug_gnoisgg33dw}
  TextureObject() {texture=&T_Tiling_Noise_01}
  VectorParameter() {parameterName=ChaparralGrassColor2, defaultValue=<rgba 0.25 0.25 0.25 1>}
 ]}
