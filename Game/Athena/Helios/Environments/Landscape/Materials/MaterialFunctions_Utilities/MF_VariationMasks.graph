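# Material function in the Helios landscape utilities that outputs four micro variation masks and one macro variation mask from tiling Gaussian noise.
# World XY, excluding shader offsets by default, is divided by per-mask scales to sample the green channel of an overridable noise texture.
# Micro masks 1–3 are remapped by min/max parameters and saturated, with mask 3 offsetting its position.
# Micro mask 4 reuses mask 2's raw sample, and the macro mask is also unremapped.
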MaterialFunction /Game/Athena/Helios/Environments/Landscape/Materials/MaterialFunctions_Utilities/MF_VariationMasks
graph {nodes [
  n1 = FunctionInput(Preview=WorldPosition() {worldPositionShaderOffset=WPT_ExcludeAllShaderOffsets}.XYZ) {bUsePreviewValueAsDefault=true, inputName=Position, sortPriority=1}
  n2 = ComponentMask(None=@n1) {g=true, r=true}
  n3 = FunctionInput(Preview=ScalarParameter() {defaultValue=5000, group=VariationMask, parameterName=MicroVariationMaskScale_01, sortPriority=-50}) {bUsePreviewValueAsDefault=true, inputName=MicroScale_01, inputType=FunctionInput_Scalar, sortPriority=2}
  n4 = LinearInterpolate(
   A=ScalarParameter() {group=VariationMask, parameterName=MicroVariationMask1_Blend_Min, sortPriority=-49}
   B=ScalarParameter() {defaultValue=1, group=VariationMask, parameterName=MicroVariationMask1_Blend_Max, sortPriority=-48}
   Alpha=TextureSample(UVs=Divide(A=@n2, B=@n3), Tex=NamedRerouteUsage()) {samplerSource=SSM_Wrap_WorldGroupSettings, texture=/Game/Effects/Fort_Effects/Textures/Tile/T_GaussianNoise_01_M}.G
  )
  FunctionOutput(None=NamedRerouteDeclaration(None=Saturate(None=@n4)) {name=MicroVariation Mask_01, nodeColor=<rgba 1 0 0.8203 1>}) {outputName=MicroVariationMask_01}
  n5 = ScalarParameter() {defaultValue=12000, group=VariationMask, parameterName=MicroVariationMaskScale_02, sortPriority=-45}
  n6 = TextureSample(
   UVs=Divide(
    A=@n2
    B=FunctionInput(Preview=@n5) {bUsePreviewValueAsDefault=true, inputName=MicroScale_02, inputType=FunctionInput_Scalar, sortPriority=3}
   )
   Tex=NamedRerouteUsage()
  ) {samplerSource=SSM_Wrap_WorldGroupSettings, texture=&T_GaussianNoise_01_M}
  n7 = LinearInterpolate(
   A=ScalarParameter() {group=VariationMask, parameterName=MicroVariationMask2_Blend_Min, sortPriority=-44}
   B=ScalarParameter() {defaultValue=1, group=VariationMask, parameterName=MicroVariationMask2_Blend_Max, sortPriority=-43}
   Alpha=@n6.G
  )
  FunctionOutput(None=NamedRerouteDeclaration(None=Saturate(None=@n7)) {name=MicroVariation Mask_02, nodeColor=<rgba 1 0 0.8203 1>}) {outputName=MicroVariationMask_02, sortPriority=1}
  n8 = Add(A=@n2, B=Constant() {r=500})
  n9 = FunctionInput(Preview=ScalarParameter() {defaultValue=10000, group=VariationMask, parameterName=MicroVariationMaskScale_03, sortPriority=-40}) {bUsePreviewValueAsDefault=true, inputName=MicroScale_03, inputType=FunctionInput_Scalar, sortPriority=4}
  n10 = LinearInterpolate(
   A=ScalarParameter() {group=VariationMask, parameterName=MicroVariationMask3_Blend_Min, sortPriority=-39}
   B=ScalarParameter() {defaultValue=1, group=VariationMask, parameterName=MicroVariationMask3_Blend_Max, sortPriority=-38}
   Alpha=TextureSample(UVs=Divide(A=@n8, B=@n9), Tex=NamedRerouteUsage()) {samplerSource=SSM_Wrap_WorldGroupSettings, texture=&T_GaussianNoise_01_M}.G
  )
  FunctionOutput(None=NamedRerouteDeclaration(None=Saturate(None=@n10)) {name=MicroVariation Mask_03, nodeColor=<rgba 1 0 0.8203 1>}) {outputName=MicroVariationMask_03, sortPriority=2}
  FunctionOutput(None=NamedRerouteDeclaration(None=@n6.G) {name=MicroVariation Mask_04, nodeColor=<rgba 1 0 0.6562 1>}) {outputName=MicroVariationMask_04, sortPriority=3}
  n11 = Divide(
   A=ComponentMask(None=@n1) {g=true, r=true}
   B=FunctionInput(Preview=ScalarParameter() {defaultValue=25000, group=VariationMask, parameterName=MacroVariationMaskScale, sortPriority=-30}) {bUsePreviewValueAsDefault=true, inputName=MacroScale, inputType=FunctionInput_Scalar, sortPriority=5}
  )
  n12 = NamedRerouteUsage()
  FunctionOutput(None=NamedRerouteDeclaration(None=TextureSample(UVs=@n11, Tex=@n12) {samplerSource=SSM_Wrap_WorldGroupSettings, texture=&T_GaussianNoise_01_M}.G) {name=MacroVariationMask, nodeColor=<rgba 1 0.9609 0 1>}) {outputName=MacroVariationMask, sortPriority=4}
  NamedRerouteDeclaration(None=FunctionInput(Preview=TextureObject() {texture=&T_GaussianNoise_01_M}) {bUsePreviewValueAsDefault=true, inputName=NoiseTexture, inputType=FunctionInput_Texture2D}) {name=NoiseTexture, nodeColor=<rgba 0 1 0.5547 1>}
 ]}
