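# Material function that adds a mobile wet-shore waterline to incoming landscape material attributes, darkening and smoothing surfaces near the water.
# World Z plus a per-primitive height offset builds a gradient mask, also output separately, that blends tinted base color and wet roughness into dry values.
# A Max term pushes roughness back up below the waterline.
# A static switch disables the effect, leaving only the roughness offset.
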MaterialFunction /Game/Athena/Helios/Environments/Landscape/Materials/MaterialFunctions_Utilities/MF_WetShore_Mobile
graph {nodes [
  n1 = GetMaterialAttributes(None=FunctionInput() {inputName=Material, inputType=FunctionInput_MaterialAttributes}) {attributeGetTypes [#1, #2]}
  n2 = VectorParameter() {group=Mobile WaterLine, parameterName=Mobile WaterLine Color, defaultValue=<rgba 0.4137 0.4531 0.203 1>}
  n3 = WorldPosition()
  n4 = ScalarParameter() {
   bUseCustomPrimitiveData=true
   desc="Primitive Data Only, Enable per instance in the details panel."
   group=Waterline
   parameterName=PD_WaterlineHeightOffset
   sortPriority=0
  }
  n5 = Add(A=Add(A=TruncateLWC!(None=@n3.Z), B=@n4)) {constB=60}
  n6 = ScalarParameter() {defaultValue=60, group=Mobile WaterLine, parameterName=Mobile WaterLine Length}
  n7 = Divide(A=@n5, B=@n6)
  n8 = FunctionOutput(None=Saturate(None=@n7)) {outputName=Waterline Height Mask, sortPriority=100}
  n9 = LinearInterpolate(A=Multiply(A=@n1.BaseColor, B=@n2.RGB), B=@n1.BaseColor, Alpha=@n8)
  n10 = Add(A=@n1.Roughness, B=ScalarParameter() {group=Mobile WaterLine, parameterName=Mobile WaterLine Roughness Offset})
  n11 = Max(
   A=LinearInterpolate(
    A=ScalarParameter() {defaultValue=0.05, group=Mobile WaterLine, parameterName=Mobile WaterLine Roughness Wet}
    B=Saturate(None=@n10)
    Alpha=@n8
   ) {constA=0.05, constB=0}
   B=Saturate(None=Multiply(A=@n5) {constB=-0.05})
  )
  FunctionOutput(None=StaticSwitch(
    True=SetMaterialAttributes!(MaterialAttributes=@n1.MaterialAttributes, BaseColor=@n9, Roughness=@n11) {attributeSetTypes [#1, #2]}
    False=SetMaterialAttributes!(MaterialAttributes=@n1.MaterialAttributes, Roughness=@n10) {attributeSetTypes [#2]}
    Value=StaticBoolParameter() {defaultValue=true, group=Mobile WaterLine, parameterName=Use Mobile WaterLine}
   ))
  Multiply(A=@n7, B=@n6) {constB=0.005}
 ]}
